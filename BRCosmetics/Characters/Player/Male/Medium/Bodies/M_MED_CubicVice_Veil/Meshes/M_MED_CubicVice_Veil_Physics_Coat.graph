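# Physics asset for the CubicVice Veil medium male body that sets up secondary motion for the coat and two ropes.
# Left and right front, side and back coat chains and the rope chains hang from small base bodies on the pelvis, with angular limits along each chain.
# Free-angle cross constraints link neighboring coat panels at their lower segments.
# Pelvis and leg capsules plus a flat root box supply body and floor-level collision shapes.
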PhysicsAsset /BRCosmetics/Characters/Player/Male/Medium/Bodies/M_MED_CubicVice_Veil/Meshes/M_MED_CubicVice_Veil_Physics_Coat
bodies {
 BASE_L_rope [{shapeType=Capsule, length=0.505, radius=0.505, shapeName=BASE_L_rope_capsule, rotation=<rot 0 -90 0>}]
 BASE_R_rope [{shapeType=Capsule, length=0.505, radius=0.505, shapeName=BASE_R_rope_capsule, rotation=<rot 0 -90 0>}]
 L_rope_01 [{
   shapeType=Capsule
   length=2.5222
   radius=2.4314
   shapeName=L_rope_01_capsule
   center=<3.0597 -0.1571 -0.1014>
   rotation=<rot 0.3755 -90 -84.0784>
  }]
 L_rope_02 [{
   shapeType=Capsule
   length=3.1844
   radius=2.8452
   shapeName=L_rope_02_capsule
   center=<3.4997 -0.3724 0.2319>
   rotation=<rot 0 -83.9766 -87.0087>
  }]
 R_rope_01 [{
   shapeType=Capsule
   length=3.9369
   radius=2.4139
   shapeName=R_rope_01_capsule
   center=<3.6237 0.0387 0.0881>
   rotation=<rot -21.777 -89.9999 93.1511>
  }]
 R_rope_02 [{
   shapeType=Capsule
   length=3.073
   radius=2.7695
   shapeName=R_rope_02_capsule
   center=<4.4064 -0.0029 -0.4674>
   rotation=<rot 2.4295 -88.0287 -85.5552>
  }]
 calf_l [
  {shapeType=Capsule, length=35.654, radius=10.5695, center=<-28.8055 0.2209 3.6476>, rotation=<rot -3.3016 86.6217 -92.3669>}
  {
   shapeType=Capsule
   length=6.4593
   radius=10.5695
   shapeName=calf_l_capsule
   center=<-46.3499 1.1049 1.4674>
   rotation=<rot -3.3016 86.6216 -92.7659>
  }
 ]
 calf_r [{shapeType=Capsule, length=29.2325, radius=13.2321, center=<26.8216 -7.6461 -2.2715>, rotation=<rot 0.0868 87.4652 -88.4455>}]
 dyn_L_backCoat_1 [{
   shapeType=Capsule
   length=9.9957
   radius=8.0043
   shapeName=dyn_L_backCoat_1_capsule
   center=<7.1292 0.4195 7.3761>
   rotation=<rot -7.7845 -1.1424 98.3756>
  }]
 dyn_L_backCoat_2 [{
   shapeType=Capsule
   length=14.6098
   radius=7.5
   shapeName=dyn_L_backCoat_2_capsule
   center=<4.6653 -0.6342 1.7807>
   rotation=<rot -5.2684 -0.6961 99.131>
  }]
 dyn_L_backCoat_3 [{
   shapeType=Capsule
   length=15.279
   radius=8
   shapeName=dyn_L_backCoat_3_capsule
   center=<4.2235 0.5164 3.953>
   rotation=<rot -8.0513 -1.182 98.3811>
  }]
 dyn_L_backCoat_base [{shapeType=Capsule, length=0.505, radius=0.505, shapeName=dyn_L_backCoat_base_capsule, rotation=<rot 0 -90 0>}]
 dyn_L_frontCoat_1 [{
   shapeType=Capsule
   length=11.027
   radius=7.473
   shapeName=dyn_L_frontCoat_1_capsule
   center=<6.8405 0.908 6.1664>
   rotation=<rot -7.1216 4.9162 120.9623>
  }]
 dyn_L_frontCoat_2 [
  {
   shapeType=Capsule
   length=6.8623
   radius=14.1377
   shapeName=dyn_L_frontCoat_2_capsule
   center=<13.9475 2.4817 11.894>
   rotation=<rot -9.4735 1.685 100.2021>
  }
  {
   shapeType=Capsule
   length=14.9779
   radius=2.6719
   shapeName=dyn_L_frontCoat_2_capsule
   center=<10.9997 9.0356 -2.7932>
   rotation=<rot 10.5771 -85.2363 89.6828>
  }
 ]
 dyn_L_frontCoat_3 [
  {
   shapeType=Capsule
   length=2.5805
   radius=16.4195
   shapeName=dyn_L_frontCoat_3_capsule
   center=<11.7207 2.7012 10.6817>
   rotation=<rot -5.9998 2.7608 98.5757>
  }
  {
   shapeType=Capsule
   length=13.0825
   radius=2.6595
   shapeName=dyn_L_frontCoat_3_capsule
   center=<10.179 11.0775 -2.8846>
   rotation=<rot 8.8718 -89.7516 89.0831>
  }
 ]
 dyn_L_frontCoat_base [{
   shapeType=Capsule
   length=7.7871
   radius=6.4841
   shapeName=dyn_L_frontCoat_1_capsule
   center=<6.9385 0.5389 4.3361>
   rotation=<rot 4.0917 -174.546 74.3827>
  }]
 dyn_L_frontCoat_base_BASE [{shapeType=Capsule, length=0.505, radius=0.505, shapeName=dyn_L_frontCoat_base_BASE_capsule, rotation=<rot 0 -90 0>}]
 dyn_L_sideCoat_1 [{
   shapeType=Capsule
   length=9.246
   radius=8.754
   shapeName=dyn_L_sideCoat_1_capsule
   center=<8.0863 -0.3307 7.5646>
   rotation=<rot 0.0792 2.973 87.9309>
  }]
 dyn_L_sideCoat_2 [{shapeType=Capsule, length=13.1098, radius=9, shapeName=dyn_L_sideCoat_2_capsule, center=<5.8359 -0.2975 4.6466>, rotation=<rot 0 0 90>}]
 dyn_L_sideCoat_base [{shapeType=Capsule, length=0.505, radius=0.505, shapeName=dyn_L_sideCoat_base_capsule, rotation=<rot 0 -90 0>}]
 dyn_R_backCoat_1 [{
   shapeType=Capsule
   length=10.6945
   radius=7.3055
   shapeName=dyn_R_backCoat_1_capsule
   center=<-7.3531 -0.0769 -6.4518>
   rotation=<rot -8.6337 -1.2016 104.3026>
  }]
 dyn_R_backCoat_2 [{
   shapeType=Capsule
   length=15.1098
   radius=7
   shapeName=dyn_R_backCoat_2_capsule
   center=<-4.1551 0.9479 -2.7275>
   rotation=<rot -8.8352 -1.1355 107.8954>
  }]
 dyn_R_backCoat_3 [{
   shapeType=Capsule
   length=18.779
   radius=8
   shapeName=dyn_R_backCoat_3_capsule
   center=<-3.5395 -2.5727 -4.8773>
   rotation=<rot -8.2394 -1.0579 107.3492>
  }]
 dyn_R_backCoat_base [{shapeType=Capsule, length=0.505, radius=0.505, shapeName=dyn_R_backCoat_base_capsule, rotation=<rot 0 -90 0>}]
 dyn_R_frontCoat_1 [
  {
   shapeType=Capsule
   length=14.0619
   radius=13.6605
   shapeName=dyn_L_frontCoat_1_capsule
   center=<-8.2107 -3.8492 -12.1109>
   rotation=<rot -11.6977 4.3384 -72.4533>
  }
  {
   shapeType=Capsule
   length=11.681
   radius=8.2137
   shapeName=dyn_R_frontCoat_1_capsule
   center=<-12.9208 -4.85 -6.5144>
   rotation=<rot -11.6977 4.3384 -62.9169>
  }
  {
   shapeType=Capsule
   length=17.2881
   radius=2.6067
   shapeName=dyn_R_frontCoat_3_capsule
   center=<-9.4682 -6.013 1.3893>
   rotation=<rot -26.6658 95.693 -84.6622>
  }
 ]
 dyn_R_frontCoat_2 [{
   shapeType=Capsule
   length=5.743
   radius=12.257
   shapeName=dyn_L_frontCoat_2_capsule
   center=<-10.0261 -4.4104 -11.3332>
   rotation=<rot -9.4736 1.6849 -69.7979>
  }]
 dyn_R_frontCoat_3 [{
   shapeType=Capsule
   length=6.2495
   radius=11.7505
   shapeName=dyn_R_frontCoat_3_capsule
   center=<-7.1821 -4.7797 -9.3013>
   rotation=<rot -10.8451 3.1348 109.8353>
  }]
 dyn_R_frontCoat_base [{
   shapeType=Capsule
   length=7.7871
   radius=6.4841
   shapeName=dyn_L_frontCoat_1_capsule
   center=<-6.1939 -2.6996 -2.6412>
   rotation=<rot 4.0917 -174.5457 58.1198>
  }]
 dyn_R_frontCoat_base_BASE [{shapeType=Capsule, length=0.505, radius=0.505, shapeName=dyn_R_frontCoat_base_BASE_capsule, rotation=<rot 0 -90 0>}]
 dyn_R_sideCoat_1 [{shapeType=Capsule, length=10.7242, radius=6.4414, shapeName=dyn_R_sideCoat_1_capsule, center=<-9.6171 0 -3.8137>, rotation=<rot 0 0 90>}]
 dyn_R_sideCoat_2 [
  {
   shapeType=Capsule
   length=14.2831
   radius=7.8267
   shapeName=dyn_R_sideCoat_2_capsule
   center=<-7.2415 0 -6.1092>
   rotation=<rot -8.7132 0 90>
  }
  {
   shapeType=Capsule
   length=14.2831
   radius=7.8267
   shapeName=dyn_R_sideCoat_2_capsule
   center=<-14.3684 0 -3.4843>
   rotation=<rot -8.7132 0 90>
  }
 ]
 dyn_R_sideCoat_base [{shapeType=Capsule, length=0.505, radius=0.505, shapeName=dyn_R_sideCoat_base_capsule, rotation=<rot 0 -90 0>}]
 foot_l [{shapeType=Capsule, length=17.0942, radius=6.9496, center=<-5.9437 8.9001 -0.1657>, rotation=<rot 0.0002 1.7852 -89.9965>}]
 foot_r [{shapeType=Capsule, length=15.6173, radius=6.9496, center=<5.6265 -6.338 0.1656>, rotation=<rot 0.0004 -0.5613 -89.9965>}]
 pelvis [{shapeType=Capsule, length=8.4465, radius=12.5368, shapeName=pelvis_capsule, center=<0 0.8158 0>, rotation=<rot 0 -0.0002 0>}]
 root [{shapeType=Box, extentX=113.5, extentY=100.5, extentZ=7.5, shapeName=root_box, center=<0 0 -4.4019>}]
 thigh_l [
  {shapeType=Capsule, length=31.7403, radius=8.9176, center=<-18.0477 -0.4039 0.6329>, rotation=<rot 0.0001 -86.3807 -88.0225>}
  {
   shapeType=Capsule
   length=12.9084
   radius=10.4775
   shapeName=thigh_l_capsule
   center=<-25.6847 -0.5958 0.6306>
   rotation=<rot 0.1799 -91.5937 -88.0307>
  }
 ]
 thigh_r [{
   shapeType=Capsule
   length=3.6538
   radius=2.4527
   shapeName=thigh_r_capsule
   center=<0.0496 0 -0.7237>
   rotation=<rot 86.0775 179.9998 179.9999>
  }]
}
constraints [
 {
  bone1Name=thigh_l
  bone2Name=pelvis
  swing1LimitDegrees=45
  swing1Motion=Free
  swing2LimitDegrees=45
  swing2Motion=Free
  twistLimitDegrees=45
  twistMotion=Free
 }
 {
  bone1Name=calf_l
  bone2Name=thigh_l
  swing1LimitDegrees=45
  swing1Motion=Free
  swing2LimitDegrees=45
  swing2Motion=Free
  twistLimitDegrees=45
  twistMotion=Free
 }
 {bone1Name=foot_l, bone2Name=calf_l, swing1LimitDegrees=45, swing2LimitDegrees=45, twistLimitDegrees=45}
 {
  bone1Name=thigh_r
  bone2Name=pelvis
  swing1LimitDegrees=45
  swing1Motion=Free
  swing2LimitDegrees=45
  swing2Motion=Free
  twistLimitDegrees=45
  twistMotion=Free
 }
 {
  bone1Name=calf_r
  bone2Name=thigh_r
  swing1LimitDegrees=45
  swing1Motion=Free
  swing2LimitDegrees=45
  swing2Motion=Free
  twistLimitDegrees=45
  twistMotion=Free
 }
 {bone1Name=foot_r, bone2Name=calf_r, swing1LimitDegrees=45, swing2LimitDegrees=45, twistLimitDegrees=45}
 {bone1Name=dyn_R_sideCoat_1, bone2Name=dyn_R_sideCoat_base, swing1LimitDegrees=5, swing2LimitDegrees=20, twistLimitDegrees=20}
 {bone1Name=dyn_L_sideCoat_1, bone2Name=dyn_L_sideCoat_base, swing1LimitDegrees=5, swing2LimitDegrees=20, twistLimitDegrees=20}
 {bone1Name=dyn_L_backCoat_2, bone2Name=dyn_L_backCoat_1, swing1LimitDegrees=5, swing2LimitDegrees=30, twistLimitDegrees=2}
 {bone1Name=dyn_L_backCoat_3, bone2Name=dyn_L_backCoat_2, swing1LimitDegrees=10, swing2LimitDegrees=30, twistLimitDegrees=5}
 {bone1Name=dyn_L_frontCoat_1, bone2Name=dyn_L_frontCoat_base, swing1LimitDegrees=10, swing2LimitDegrees=45, twistLimitDegrees=20}
 {bone1Name=dyn_L_frontCoat_2, bone2Name=dyn_L_frontCoat_1, swing1LimitDegrees=5, swing2LimitDegrees=50, twistLimitDegrees=5}
 {bone1Name=dyn_L_frontCoat_3, bone2Name=dyn_L_frontCoat_2, swing1LimitDegrees=10, swing2LimitDegrees=20, twistLimitDegrees=5}
 {bone1Name=dyn_L_sideCoat_2, bone2Name=dyn_L_sideCoat_1, swing1LimitDegrees=5, swing2LimitDegrees=15, twistLimitDegrees=20}
 {bone1Name=dyn_R_backCoat_2, bone2Name=dyn_R_backCoat_1, swing1LimitDegrees=5, swing2LimitDegrees=30, twistLimitDegrees=2}
 {bone1Name=dyn_R_backCoat_3, bone2Name=dyn_R_backCoat_2, swing1LimitDegrees=10, swing2LimitDegrees=30, twistLimitDegrees=5}
 {bone1Name=dyn_R_frontCoat_1, bone2Name=dyn_R_frontCoat_base, swing1LimitDegrees=10, swing2LimitDegrees=45, twistLimitDegrees=10}
 {bone1Name=dyn_R_frontCoat_2, bone2Name=dyn_R_frontCoat_1, swing1LimitDegrees=5, swing2LimitDegrees=15, twistLimitDegrees=10}
 {bone1Name=dyn_R_frontCoat_3, bone2Name=dyn_R_frontCoat_2, swing1LimitDegrees=10, swing2LimitDegrees=20, twistLimitDegrees=5}
 {bone1Name=dyn_R_sideCoat_2, bone2Name=dyn_R_sideCoat_1, swing1LimitDegrees=10, swing2LimitDegrees=45, twistLimitDegrees=15}
 {
  bone1Name=dyn_L_sideCoat_base
  bone2Name=pelvis
  swing1LimitDegrees=45
  swing1Motion=Free
  swing2LimitDegrees=45
  swing2Motion=Free
  twistLimitDegrees=45
  twistMotion=Free
 }
 {
  bone1Name=dyn_R_sideCoat_base
  bone2Name=pelvis
  swing1LimitDegrees=45
  swing1Motion=Free
  swing2LimitDegrees=45
  swing2Motion=Free
  twistLimitDegrees=45
  twistMotion=Free
 }
 {bone1Name=dyn_R_backCoat_1, bone2Name=dyn_R_backCoat_base, swing1LimitDegrees=10, swing2LimitDegrees=20, twistLimitDegrees=2}
 {
  bone1Name=dyn_R_backCoat_base
  bone2Name=pelvis
  swing1LimitDegrees=45
  swing1Motion=Free
  swing2LimitDegrees=45
  swing2Motion=Free
  twistLimitDegrees=45
  twistMotion=Free
 }
 {
  bone1Name=dyn_R_sideCoat_2
  bone2Name=dyn_R_frontCoat_2
  swing1LimitDegrees=45
  swing1Motion=Free
  swing2LimitDegrees=45
  swing2Motion=Free
  twistLimitDegrees=45
  twistMotion=Free
 }
 {
  bone1Name=dyn_L_sideCoat_2
  bone2Name=dyn_L_frontCoat_2
  swing1LimitDegrees=45
  swing1Motion=Free
  swing2LimitDegrees=45
  swing2Motion=Free
  twistLimitDegrees=45
  twistMotion=Free
 }
 {bone1Name=dyn_L_frontCoat_base, bone2Name=dyn_L_frontCoat_base_BASE, swing1LimitDegrees=10, swing2LimitDegrees=70, twistLimitDegrees=70}
 {
  bone1Name=dyn_L_frontCoat_base_BASE
  bone2Name=pelvis
  swing1LimitDegrees=45
  swing1Motion=Free
  swing2LimitDegrees=45
  swing2Motion=Free
  twistLimitDegrees=45
  twistMotion=Free
 }
 {bone1Name=dyn_R_frontCoat_base, bone2Name=dyn_R_frontCoat_base_BASE, swing1LimitDegrees=10, swing2LimitDegrees=70, twistLimitDegrees=70}
 {
  bone1Name=dyn_R_frontCoat_base_BASE
  bone2Name=pelvis
  swing1LimitDegrees=45
  swing1Motion=Free
  swing2LimitDegrees=45
  swing2Motion=Free
  twistLimitDegrees=45
  twistMotion=Free
 }
 {
  bone1Name=BASE_L_rope
  bone2Name=pelvis
  swing1LimitDegrees=45
  swing1Motion=Free
  swing2LimitDegrees=45
  swing2Motion=Free
  twistLimitDegrees=45
  twistMotion=Free
 }
 {bone1Name=L_rope_01, bone2Name=BASE_L_rope, swing1LimitDegrees=30, swing2LimitDegrees=30, twistLimitDegrees=10}
 {bone1Name=L_rope_02, bone2Name=L_rope_01, swing1LimitDegrees=30, swing2LimitDegrees=30, twistLimitDegrees=10}
 {
  bone1Name=BASE_R_rope
  bone2Name=pelvis
  swing1LimitDegrees=45
  swing1Motion=Free
  swing2LimitDegrees=45
  swing2Motion=Free
  twistLimitDegrees=45
  twistMotion=Free
 }
 {bone1Name=R_rope_01, bone2Name=BASE_R_rope, swing1LimitDegrees=30, swing2LimitDegrees=30, twistLimitDegrees=10}
 {bone1Name=R_rope_02, bone2Name=R_rope_01, swing1LimitDegrees=30, swing2LimitDegrees=30, twistLimitDegrees=10}
 {bone1Name=dyn_L_backCoat_1, bone2Name=dyn_L_backCoat_base, swing1LimitDegrees=10, swing2LimitDegrees=20, twistLimitDegrees=2}
 {
  bone1Name=dyn_L_backCoat_base
  bone2Name=pelvis
  swing1LimitDegrees=45
  swing1Motion=Free
  swing2LimitDegrees=45
  swing2Motion=Free
  twistLimitDegrees=45
  twistMotion=Free
 }
 {
  bone1Name=dyn_R_backCoat_3
  bone2Name=dyn_R_sideCoat_2
  swing1LimitDegrees=45
  swing1Motion=Free
  swing2LimitDegrees=45
  swing2Motion=Free
  twistLimitDegrees=45
  twistMotion=Free
 }
 {
  bone1Name=dyn_L_backCoat_3
  bone2Name=dyn_L_sideCoat_2
  swing1LimitDegrees=45
  swing1Motion=Free
  swing2LimitDegrees=45
  swing2Motion=Free
  twistLimitDegrees=45
  twistMotion=Free
 }
]
