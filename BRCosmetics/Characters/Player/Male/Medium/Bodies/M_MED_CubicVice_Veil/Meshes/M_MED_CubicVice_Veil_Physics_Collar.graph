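# Physics asset for the collar of the medium male CubicVice Veil body, with capsule bodies for the collar bones and the surrounding upper body.
# Seven collar chains (left A–C, middle, right A–C) each run from a tiny BASE body on spine_05 through two segments, using tight angle limits.
# Larger capsules cover head, upper spine, clavicles, arms and hands.
# Head, spine and zipper pin constraints leave swing and twist free.
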PhysicsAsset /BRCosmetics/Characters/Player/Male/Medium/Bodies/M_MED_CubicVice_Veil/Meshes/M_MED_CubicVice_Veil_Physics_Collar
bodies {
 BASE_L_A_collar_01 [{shapeType=Capsule, length=0.505, radius=0.505, shapeName=BASE_L_A_collar_01_capsule, rotation=<rot 0 -90 0>}]
 BASE_L_B_collar_01 [{shapeType=Capsule, length=0.505, radius=0.505, shapeName=BASE_L_B_collar_01_capsule, rotation=<rot 0 -90 0>}]
 BASE_L_C_collar_01 [{shapeType=Capsule, length=0.505, radius=0.505, shapeName=BASE_L_C_collar_01_capsule, rotation=<rot 0 -90 0>}]
 BASE_M_collar_01 [{shapeType=Capsule, length=0.505, radius=0.505, shapeName=BASE_M_collar_01_capsule, rotation=<rot 0 -90 0>}]
 BASE_R_A_collar_01 [{shapeType=Capsule, length=0.505, radius=0.505, shapeName=BASE_R_A_collar_01_capsule, rotation=<rot 0 -90 0>}]
 BASE_R_B_collar_01 [{shapeType=Capsule, length=0.505, radius=0.505, shapeName=BASE_R_B_collar_01_capsule, rotation=<rot 0 -90 0>}]
 BASE_R_C_collar_01 [{shapeType=Capsule, length=0.505, radius=0.505, shapeName=BASE_R_C_collar_01_capsule, rotation=<rot 0 -90 0>}]
 L_A_collar_01 [{
   shapeType=Capsule
   length=2.6741
   radius=2.215
   shapeName=L_A_collar_01_capsule
   center=<2.9942 1.6906 0.3009>
   rotation=<rot -1.8009 -82.7165 -85.5808>
  }]
 L_A_collar_02 [{
   shapeType=Capsule
   length=5.1918
   radius=3.0551
   shapeName=L_A_collar_02_capsule
   center=<4.1816 0.8257 1.0112>
   rotation=<rot 0 1.0712 -78.618>
  }]
 L_B_collar_01 [{
   shapeType=Capsule
   length=4.9203
   radius=1.6405
   shapeName=L_B_collar_01_capsule
   center=<2.452 0.6359 0.0686>
   rotation=<rot 1.7451 174.9965 -87.8495>
  }]
 L_B_collar_02 [{
   shapeType=Capsule
   length=4.7312
   radius=2.6788
   shapeName=L_B_collar_02_capsule
   center=<1.8957 0.6538 0.1199>
   rotation=<rot 2.7545 -11.6022 -86.32>
  }]
 L_C_collar_01 [{
   shapeType=Capsule
   length=3.3758
   radius=1.8
   shapeName=L_C_collar_01_capsule
   center=<1.8268 -0.2034 -0.4138>
   rotation=<rot -1.5002 2.3049 -76.1859>
  }]
 L_C_collar_02 [{
   shapeType=Capsule
   length=5.1089
   radius=2.1285
   shapeName=L_C_collar_02_capsule
   center=<1.7173 0.8933 0.0519>
   rotation=<rot 0 3.3345 -85.7821>
  }]
 M_collar_01 [{shapeType=Capsule, length=2.8243, radius=1.7582, shapeName=M_collar_01_capsule, center=<1.8709 0 -0.1748>, rotation=<rot 0 -180 -90>}]
 M_collar_02 [{shapeType=Capsule, length=4.3657, radius=2.0168, shapeName=M_collar_02_capsule, center=<1.6806 0 0.322>, rotation=<rot 0 -180 -90>}]
 R_A_collar_01 [{
   shapeType=Capsule
   length=3.4593
   radius=2.3777
   shapeName=R_A_collar_01_capsule
   center=<-3.4055 -0.5758 -0.2593>
   rotation=<rot -1.8214 -86.4418 -87.1539>
  }]
 R_A_collar_02 [{
   shapeType=Capsule
   length=4.2742
   radius=3.5443
   shapeName=R_A_collar_02_capsule
   center=<-4.601 -0.5966 -1.9117>
   rotation=<rot 0 -0.254 -79.3089>
  }]
 R_B_collar_01 [{
   shapeType=Capsule
   length=3.1247
   radius=1.9976
   shapeName=R_B_collar_01_capsule
   center=<-3.018 0.0632 -0.0251>
   rotation=<rot 1.9151 170.3046 -87.9995>
  }]
 R_B_collar_02 [{
   shapeType=Capsule
   length=3.6817
   radius=1.765
   shapeName=R_B_collar_02_capsule
   center=<-2.8222 0.3379 -0.1791>
   rotation=<rot 2.5861 -12.2398 -89.3606>
  }]
 R_C_collar_01 [{
   shapeType=Capsule
   length=2.4309
   radius=2.2565
   shapeName=R_C_collar_01_capsule
   center=<-2.2853 -0.1068 0.4979>
   rotation=<rot 0 -3.7728 -76.1062>
  }]
 R_C_collar_02 [{
   shapeType=Capsule
   length=5.1089
   radius=2.1285
   shapeName=R_C_collar_02_capsule
   center=<-1.7174 -0.8933 -0.0519>
   rotation=<rot 0 3.3345 -85.7821>
  }]
 clavicle_l [{shapeType=Capsule, length=13.2048, radius=7.605, center=<9.2636 -0.2732 -1.3551>, rotation=<rot 2.4009 -82.6715 -74.5692>}]
 clavicle_r [{shapeType=Capsule, length=13.3404, radius=7.4694, center=<-11.7341 0.9168 1.8305>, rotation=<rot -0.0002 -75.0939 -72.359>}]
 hand_l [{shapeType=Capsule, length=8.6211, radius=7.4314, center=<7.5171 -1.5826 0.1717>, rotation=<rot 0 89.4775 -58.4079>}]
 hand_r [{shapeType=Capsule, length=8.6211, radius=7.4314, center=<-7.5171 1.5826 -0.1717>, rotation=<rot 0 89.4775 -58.4079>}]
 head [
  {shapeType=Capsule, length=6.5345, radius=6.5559, center=<3.1811 2.6041 0.3139>, rotation=<rot 0.5016 -90.415 -88.3249>}
  {
   shapeType=Capsule
   length=3.3888
   radius=6.0811
   shapeName=head_capsule_01
   center=<1.6133 -0.0375 -0.0239>
   rotation=<rot -89.1519 179.9981 -180>
  }
  {
   shapeType=Capsule
   length=6.5345
   radius=6.5559
   shapeName=head_capsule_02
   center=<9.5547 -1.6477 0.306>
   rotation=<rot 0.5016 -90.415 -88.3249>
  }
 ]
 lowerarm_l [{shapeType=Capsule, length=14.6726, radius=6.1024, center=<9.5225 0.3036 0.4043>, rotation=<rot 0 -82.8733 -88.0762>}]
 lowerarm_r [{shapeType=Capsule, length=14.6724, radius=6.1024, center=<-9.5226 -0.3038 -0.4041>, rotation=<rot 0 -82.8778 -88.0737>}]
 pin_to_geo_zipper [{shapeType=Capsule, length=0.2, radius=0.2, shapeName=pin_to_geo_zipper_capsule, rotation=<rot 0 -90 0>}]
 spine_04 [{shapeType=Capsule, length=14.805, radius=12.3, shapeName=spine_04_capsule, center=<4.2679 0.8352 0.0244>, rotation=<rot 0.3135 -90 -1.0375>}]
 spine_05 [{shapeType=Capsule, length=17.0947, radius=12.4735, center=<0.1311 -0.5277 0>}]
 upperarm_twist_01_l [{shapeType=Capsule, length=13.805, radius=6.4659, center=<9.0015 0.0131 -1.259>, rotation=<rot 8.0871 -89.9999 87.5129>}]
 upperarm_twist_01_r [{shapeType=Capsule, length=13.5559, radius=6.7343, center=<-9.6044 -0.2671 0.6211>, rotation=<rot 10.6093 -89.9999 -94.3654>}]
}
constraints [
 {bone1Name=clavicle_l, bone2Name=spine_05, swing1LimitDegrees=45, swing2LimitDegrees=45, twistLimitDegrees=45}
 {bone1Name=lowerarm_l, bone2Name=clavicle_l, swing1LimitDegrees=45, swing2LimitDegrees=45, twistLimitDegrees=45}
 {bone1Name=hand_l, bone2Name=lowerarm_l, swing1LimitDegrees=45, swing2LimitDegrees=45, twistLimitDegrees=45}
 {bone1Name=upperarm_twist_01_l, bone2Name=clavicle_l, swing1LimitDegrees=45, swing2LimitDegrees=45, twistLimitDegrees=45}
 {bone1Name=clavicle_r, bone2Name=spine_05, swing1LimitDegrees=45, swing2LimitDegrees=45, twistLimitDegrees=45}
 {bone1Name=lowerarm_r, bone2Name=clavicle_r, swing1LimitDegrees=45, swing2LimitDegrees=45, twistLimitDegrees=45}
 {bone1Name=hand_r, bone2Name=lowerarm_r, swing1LimitDegrees=45, swing2LimitDegrees=45, twistLimitDegrees=45}
 {bone1Name=upperarm_twist_01_r, bone2Name=clavicle_r, swing1LimitDegrees=45, swing2LimitDegrees=45, twistLimitDegrees=45}
 {bone1Name=BASE_L_A_collar_01, bone2Name=spine_05, swing1LimitDegrees=10, swing2LimitDegrees=45, twistLimitDegrees=5}
 {bone1Name=L_A_collar_01, bone2Name=BASE_L_A_collar_01, swing1LimitDegrees=10, swing2LimitDegrees=10, twistLimitDegrees=5}
 {bone1Name=L_A_collar_02, bone2Name=L_A_collar_01, swing1LimitDegrees=10, swing2LimitDegrees=10, twistLimitDegrees=5}
 {bone1Name=BASE_L_B_collar_01, bone2Name=spine_05, swing1LimitDegrees=10, swing2LimitDegrees=45, twistLimitDegrees=5}
 {bone1Name=L_B_collar_01, bone2Name=BASE_L_B_collar_01, swing1LimitDegrees=10, swing2LimitDegrees=10, twistLimitDegrees=5}
 {bone1Name=L_B_collar_02, bone2Name=L_B_collar_01, swing1LimitDegrees=10, swing2LimitDegrees=10, twistLimitDegrees=5}
 {bone1Name=BASE_L_C_collar_01, bone2Name=spine_05, swing1LimitDegrees=10, swing2LimitDegrees=45, twistLimitDegrees=5}
 {bone1Name=L_C_collar_01, bone2Name=BASE_L_C_collar_01, swing1LimitDegrees=10, swing2LimitDegrees=10, twistLimitDegrees=5}
 {bone1Name=L_C_collar_02, bone2Name=L_C_collar_01, swing1LimitDegrees=10, swing2LimitDegrees=10, twistLimitDegrees=5}
 {bone1Name=BASE_M_collar_01, bone2Name=spine_05, swing1LimitDegrees=10, swing2LimitDegrees=45, twistLimitDegrees=5}
 {bone1Name=M_collar_01, bone2Name=BASE_M_collar_01, swing1LimitDegrees=10, swing2LimitDegrees=10, twistLimitDegrees=5}
 {bone1Name=M_collar_02, bone2Name=M_collar_01, swing1LimitDegrees=10, swing2LimitDegrees=10, twistLimitDegrees=5}
 {bone1Name=BASE_R_A_collar_01, bone2Name=spine_05, swing1LimitDegrees=10, swing2LimitDegrees=45, twistLimitDegrees=5}
 {bone1Name=R_A_collar_01, bone2Name=BASE_R_A_collar_01, swing1LimitDegrees=3, swing2LimitDegrees=10, twistLimitDegrees=5}
 {bone1Name=R_A_collar_02, bone2Name=R_A_collar_01, swing1LimitDegrees=3, swing2LimitDegrees=10, twistLimitDegrees=5}
 {bone1Name=BASE_R_B_collar_01, bone2Name=spine_05, swing1LimitDegrees=10, swing2LimitDegrees=45, twistLimitDegrees=5}
 {bone1Name=R_B_collar_01, bone2Name=BASE_R_B_collar_01, swing1LimitDegrees=10, swing2LimitDegrees=10, twistLimitDegrees=5}
 {bone1Name=R_B_collar_02, bone2Name=R_B_collar_01, swing1LimitDegrees=10, swing2LimitDegrees=10, twistLimitDegrees=5}
 {bone1Name=BASE_R_C_collar_01, bone2Name=spine_05, swing1LimitDegrees=10, swing2LimitDegrees=45, twistLimitDegrees=5}
 {bone1Name=R_C_collar_01, bone2Name=BASE_R_C_collar_01, swing1LimitDegrees=10, swing2LimitDegrees=10, twistLimitDegrees=5}
 {bone1Name=R_C_collar_02, bone2Name=R_C_collar_01, swing1LimitDegrees=10, swing2LimitDegrees=10, twistLimitDegrees=5}
 {
  bone1Name=head
  bone2Name=spine_05
  swing1LimitDegrees=45
  swing1Motion=Free
  swing2LimitDegrees=45
  swing2Motion=Free
  twistLimitDegrees=45
  twistMotion=Free
 }
 {
  bone1Name=spine_05
  bone2Name=spine_04
  swing1LimitDegrees=45
  swing1Motion=Free
  swing2LimitDegrees=45
  swing2Motion=Free
  twistLimitDegrees=45
  twistMotion=Free
 }
 {
  bone1Name=pin_to_geo_zipper
  bone2Name=spine_05
  swing1LimitDegrees=45
  swing1Motion=Free
  swing2LimitDegrees=45
  swing2Motion=Free
  twistLimitDegrees=45
  twistMotion=Free
 }
]
