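# Pre-simulation Control Rig for the M_MED_CubicVice_Veil player body, posing coat, collar, belt, knife and rope bones.
# After a shared DeformRig function, a sequence applies corrective offsets driven by spherical pose readers on thighs, neck and arm, plus collar-to-head distance.
# Movement-state variables (crouch, slides, tactical sprint, DBNO) drive further offsets, mostly smoothed by AlphaInterp or Kalman filters.
# Parent constraints place belts, and RBAN-weighted aim constraints point front coat bones at calves.
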ControlRigBlueprint /BRCosmetics/Characters/Player/Male/Medium/Bodies/M_MED_CubicVice_Veil/Meshes/M_MED_CubicVice_Veil_presim_CtrlRig_v002
graphs [
 ContainedGraph {nodes [
   n1 = RigVMFunctionEntryNode()
   SequenceAggregate = &RigVMFunction_Sequence(ExecuteContext=@n1.ExecuteContext)
   SequenceAggregate_1 = &RigVMFunction_Sequence(ExecuteContext=@SequenceAggregate.B)
   SequenceAggregate_2 = &RigVMFunction_Sequence(ExecuteContext=@SequenceAggregate_1.B)
   SequenceAggregate_3 = &RigVMFunction_Sequence(ExecuteContext=@SequenceAggregate_2.B)
   SequenceAggregate_4 = &RigVMFunction_Sequence(ExecuteContext=@SequenceAggregate_3.B)
   SequenceAggregate_5 = &RigVMFunction_Sequence(ExecuteContext=@SequenceAggregate_4.B)
   RigVMFunctionReturnNode(
    A=@SequenceAggregate.A
    B=@SequenceAggregate_1.A
    C=@SequenceAggregate_2.A
    D=@SequenceAggregate_3.A
    E=@SequenceAggregate_4.A
    F=@SequenceAggregate_5.A
    G=@SequenceAggregate_5.B
   )
  ]}
 Multiply_4_1_ContainedGraph {nodes [
   n1 = RigVMFunctionEntryNode()
   RigVMFunctionReturnNode(Result=&RigVMFunction_MathDoubleMul(B=@n1.C, A=&RigVMFunction_MathDoubleMul(A=@n1.A, B=@n1.B).Result).Result)
  ]}
 RigVMFunctionLibrary {links [], nodes []}
 RigVMModel {nodes [
   RigUnit_BeginExecution = &RigUnit_BeginExecution()
   FN_DeformRig_FNC = RigVMFunctionReferenceNode(
    DeformRig_ Curve="(deformRig_l_upperarm_up_alpha_72_#1=False,deformRig_r_upperarm_up_alpha_73_#2=False,deformRig_l_upperarm_fr_alpha_74_#3=False,deformRig_r_upperarm_fr_alpha_75_CB51B8 ...(truncated)"
    DeformRig_Bool="(upperarm_1_#4=True,elbow_3_#5=True,wrist_7_#6=True,groin_8_#7=True,glute_17_#8=True,knee_9_40C408324CA261 ...(truncated)"
    DeformRig_Debug=(upperarm_debug_18_#4=False,wrist_debug_20_#6=False,lat_debug_26_#9=False,groinGlute_debug_27_#7=False)
    DeformRig_Value="(upperarm_up_18_#10=(Rotation=(X=0,Y=0,Z=0,W=1),Translation=(X=0,Y=0,Z=8),Scale3D=(X=0.7,Y=1,Z=1)),upperarm_fwd_fr_19_#11=(Rotation=(X=0,Y=0,Z=0,W=1),Translation=(X=0,Y=10,Z=0),Scale3D=(X=0.5,Y=1,Z= ...(truncated)"
    ExecuteContext=@RigUnit_BeginExecution.ExecutePin
   )
   > SequenceAggregate = RigVMAggregateNode()
   n1 = &RigUnit_SphericalPoseReader(
    ActiveRegionScaleFactors=(PositiveWidth=0.5,NegativeWidth=1,PositiveHeight=1,NegativeHeight=1)
    ActiveRegionSize=0.2
    Debug=(bDrawDebug=True,bDraw2D=false,bDrawLocalAxes=false,DebugScale=25,DebugSegments=20,DebugThickness=0.25)
    DriverAxis=(X=-1,Y=0,Z=0)
    DriverItem=(Type=Bone,Name="thigh_l")
    FalloffRegionScaleFactors=(PositiveWidth=0.2,NegativeWidth=0.2,PositiveHeight=0.2,NegativeHeight=0.2)
    FalloffSize=1
    FlipWidthScaling=False
    OptionalParentItem=(Type=Bone)
    RotationOffset=(X=120,Y=10,Z=0)
    ExecutePin=@SequenceAggregate.A
   )
   > &RigUnit_ModifyTransforms(
    ItemToModify=((Item=(Type=Bone,Name="dyn_L_frontCoat_base_BASE"),Transform=(Rotation=(X=0,Y=0,Z=0,W=1),Translation=(X=0,Y=-5,Z=0),Scale3D=(X=1,Y=1,Z=1))))
    ItemToModify.0.Transform.Rotation=&RigVMFunction_MathQuaternionFromEuler(Euler=(X=0,Y=0,Z=-30), RotationOrder=XYZ).Result
    Weight=@n1.OutputParam
   )
   > &RigUnit_ModifyTransforms(
    ItemToModify=((Item=(Type=Bone,Name="dyn_L_frontCoat_base_BASE"),Transform=(Rotation=(X=0,Y=-0.1736,Z=0,W=0.9848),Translation=(X=0,Y=0,Z=0),Scale3D=(X=1,Y=1,Z=1))))
    Weight=RigVMVariableNode() {variable=moving_2d}.Value
   )
   > &RigUnit_ModifyTransforms(
    ItemToModify=((Item=(Type=Bone,Name="dyn_L_sideCoat_1"),Transform=(Rotation=(X=0,Y=-0.1736,Z=0,W=0.9848),Translation=(X=0,Y=0,Z=0),Scale3D=(X=1,Y=1,Z=1))))
    Weight=&RigVMFunction_MathDoubleMul(B=RigVMVariableNode() {variable=RBAN}.Value, A=@n1.OutputParam).Result
   )
   > &RigUnit_ModifyTransforms(
    ItemToModify=((Item=(Type=Bone,Name="dyn_L_frontCoat_1"),Transform=(Rotation=(X=0,Y=-0.0872,Z=0,W=0.9962),Translation=(X=0,Y=0,Z=0),Scale3D=(X=1,Y=1,Z=1))))
    Weight=RigVMVariableNode() {variable=RBAN}.Value
   )
   > &RigUnit_ParentConstraint(
    AdvancedSettings=(RotationOrderForFilter=XZY)
    Child=(Type=Bone,Name="L_1_belt")
    Parents="((Item=(Type=Bone,Name=\"thigh_twist_01_l\"),Weight=0.004),(Item=(Type=Bone,Name=\"deform_groin_l\"),Weight=0.0052),(Item=(Type=Bone,Name=\"spine_03\"),Weight=0.0645),(Item=(Type=Bone,Name=\"spine_01\"),Weight=0.0957),(Item=(Type=Bone,Name=\"spine_02\"),Weight=0.111 ...(truncated)"
    Weight=1
   )
   > &RigUnit_ParentConstraint(
    AdvancedSettings=(RotationOrderForFilter=XZY)
    Child=(Type=Bone,Name="L_2_belt")
    Parents=((Item=(Type=Bone,Name="spine_03"),Weight=0.0763),(Item=(Type=Bone,Name="spine_01"),Weight=0.1153),(Item=(Type=Bone,Name="spine_02"),Weight=0.1365),(Item=(Type=Bone,Name="pelvis"),Weight=0.6719))
    Weight=1
   )
   > &RigUnit_ModifyTransforms(
    ItemToModify=((Item=(Type=Bone,Name="L_1_belt"),Transform=(Rotation=(X=0,Y=0,Z=0,W=1),Translation=(X=1,Y=0,Z=-4),Scale3D=(X=1,Y=1,Z=1))))
    Weight=0
   )
   > &RigUnit_ModifyTransforms(
    ItemToModify=((Item=(Type=Bone,Name="L_2_belt"),Transform=(Rotation=(X=0,Y=0,Z=0,W=1),Translation=(X=0,Y=0,Z=-2),Scale3D=(X=1,Y=1,Z=1))))
    Weight=0
   )
   KalmanFilter = &RigVMFunction_KalmanFloat(
    BufferSize=20
    Value=&RigVMFunction_MathDoubleMul(A=RigVMVariableNode() {variable=not_moving}.Value, B=RigVMVariableNode() {variable=crouching}.Value).Result
   ) {template="KalmanFilter::Execute(in Value,in BufferSize,out Result)"}
   > &RigUnit_ModifyTransforms(
    ItemToModify=((Item=(Type=Bone,Name="dyn_L_frontCoat_base_BASE"),Transform=(Rotation=(X=0,Y=0,Z=0,W=1),Translation=(X=0,Y=0,Z=-2),Scale3D=(X=1,Y=1,Z=1))))
    Weight=@KalmanFilter.Result
   )
   n2 = &RigUnit_SphericalPoseReader(
    ActiveRegionScaleFactors=(PositiveWidth=0.5,NegativeWidth=1,PositiveHeight=1,NegativeHeight=1)
    ActiveRegionSize=0.2
    Debug=(bDrawDebug=True,bDraw2D=false,bDrawLocalAxes=false,DebugScale=25,DebugSegments=20,DebugThickness=0.25)
    DriverAxis=(X=1,Y=0,Z=0)
    DriverItem=(Type=Bone,Name="thigh_r")
    FalloffRegionScaleFactors=(PositiveWidth=0.2,NegativeWidth=0.2,PositiveHeight=0.2,NegativeHeight=0.2)
    FalloffSize=1
    FlipWidthScaling=False
    OptionalParentItem=(Type=Bone)
    RotationOffset=(X=-60,Y=0,Z=0)
    ExecutePin=@SequenceAggregate.B
   )
   > &RigUnit_ModifyTransforms(
    ItemToModify=((Item=(Type=Bone,Name="dyn_R_frontCoat_base_BASE"),Transform=(Rotation=(X=0,Y=0,Z=0,W=1),Translation=(X=0,Y=0,Z=0),Scale3D=(X=1,Y=1,Z=1))))
    Weight=@n2.OutputParam
   )
   > n3 = &RigUnit_SphericalPoseReader(
    ActiveRegionScaleFactors=(PositiveWidth=0.5,NegativeWidth=1,PositiveHeight=1,NegativeHeight=1)
    ActiveRegionSize=0.2
    Debug=(bDrawDebug=True,bDraw2D=false,bDrawLocalAxes=false,DebugScale=25,DebugSegments=20,DebugThickness=0.25)
    DriverAxis=(X=-1,Y=0,Z=0)
    DriverItem=(Type=Bone,Name="calf_r")
    FalloffRegionScaleFactors=(PositiveWidth=0.2,NegativeWidth=0.2,PositiveHeight=0.2,NegativeHeight=0.2)
    FalloffSize=1
    FlipWidthScaling=False
    OptionalParentItem=(Type=Bone)
    RotationOffset=(X=-10,Y=-90,Z=0)
   )
   n4 = RigVMVariableNode() {variable=crouching}
   > &RigUnit_ModifyTransforms(
    ItemToModify=((Item=(Type=Bone,Name="dyn_R_backCoat_2"),Transform=(Rotation=(X=0.0189,Y=-0.0851,Z=0.2156,W=0.9726),Translation=(X=0,Y=0,Z=0),Scale3D=(X=1,Y=1,Z=1))))
    Weight=@n4.Value
   )
   n5 = &RigUnit_SphericalPoseReader(
    ActiveRegionScaleFactors=(PositiveWidth=3,NegativeWidth=3,PositiveHeight=2,NegativeHeight=1)
    ActiveRegionSize=0.04
    Debug=(bDrawDebug=True,bDraw2D=False,bDrawLocalAxes=False,DebugScale=25,DebugSegments=20,DebugThickness=0.25)
    DriverAxis=(X=1,Y=0,Z=0)
    DriverItem=(Type=Bone,Name="neck_01")
    FalloffRegionScaleFactors=(PositiveWidth=2,NegativeWidth=2,PositiveHeight=1,NegativeHeight=0.85)
    FalloffSize=0.15
    FlipHeightScaling=False
    FlipWidthScaling=False
    OptionalParentItem=(Type=Bone,Name="None")
    RotationOffset=(X=217.1579,Y=98.4799,Z=0.6234)
    ExecutePin=@SequenceAggregate.C
   )
   > /Script/ControlRig.RigUnit_SetRelativeTransformForItem(
    Child=(Type=Bone,Name="BASE_M_collar_01")
    Parent=(Type=Bone,Name="BASE_M_collar_01")
    Value=(Rotation=(X=0,Y=0.1736,Z=0,W=0.9848),Translation=(X=0,Y=-1.1,Z=-7),Scale3D=(X=1,Y=1,Z=1))
    bParentInitial=False
    bPropagateToChildren=True
    Weight=@n5.OutputParam
   ) {
    is=struct RigVMUnitNode
    template="Set Relative Transform::Execute(in Child,in Parent,in bParentInitial,in Value,in Weight,in bPropagateToChildren)"
   }
   > &RigUnit_SetRelativeTransformForItem(
    Child=(Type=Bone,Name="extra_M_collar_01")
    Parent=(Type=Bone,Name="extra_M_collar_01")
    Value=(Rotation=(X=0,Y=0,Z=0,W=1),Translation=(X=0,Y=0,Z=0),Scale3D=(X=1,Y=1,Z=1))
    Weight=1
    bParentInitial=False
    bPropagateToChildren=True
   ) {template="Set Relative Transform::Execute(in Child,in Parent,in bParentInitial,in Value,in Weight,in bPropagateToChildren)"}
   > &RigUnit_SetRelativeTransformForItem(
    Child=(Type=Bone,Name="extra_L_C_collar_01")
    Parent=(Type=Bone,Name="extra_L_C_collar_01")
    Value=(Rotation=(X=0,Y=0,Z=0,W=1),Translation=(X=0,Y=0,Z=-2),Scale3D=(X=1,Y=1,Z=1))
    bParentInitial=False
    bPropagateToChildren=True
    Weight=@n5.OutputParam
   ) {template="Set Relative Transform::Execute(in Child,in Parent,in bParentInitial,in Value,in Weight,in bPropagateToChildren)"}
   > n6 = &RigUnit_SphericalPoseReader(
    ActiveRegionScaleFactors=(PositiveWidth=2,NegativeWidth=2,PositiveHeight=2,NegativeHeight=2)
    ActiveRegionSize=0.05
    Debug=(bDrawDebug=True,bDraw2D=False,bDrawLocalAxes=False,DebugScale=25,DebugSegments=20,DebugThickness=0.25)
    DriverAxis=(X=1,Y=0,Z=0)
    DriverItem=(Type=Bone,Name="neck_01")
    FalloffRegionScaleFactors=(PositiveWidth=2,NegativeWidth=2,PositiveHeight=2,NegativeHeight=2)
    FalloffSize=0.13
    FlipHeightScaling=False
    FlipWidthScaling=False
    OptionalParentItem=(Type=Bone,Name="None")
    RotationOffset=(X=173.148,Y=121.5225,Z=-31.4159)
   )
   > &RigUnit_SetRelativeTransformForItem(
    Child=(Type=Bone,Name="BASE_R_C_collar_01")
    Parent=(Type=Bone,Name="BASE_R_C_collar_01")
    Value=(Rotation=(X=0,Y=0.5,Z=0,W=0.866),Translation=(X=0,Y=0,Z=0),Scale3D=(X=1,Y=1,Z=1))
    bParentInitial=False
    bPropagateToChildren=True
    Weight=@n6.OutputParam
   ) {template="Set Relative Transform::Execute(in Child,in Parent,in bParentInitial,in Value,in Weight,in bPropagateToChildren)"}
   > &RigUnit_SetRelativeTransformForItem(
    Child=(Type=Bone,Name="extra_M_collar_01")
    Parent=(Type=Bone,Name="extra_M_collar_01")
    Value=(Rotation=(X=0,Y=0,Z=0,W=1),Translation=(X=0,Y=0,Z=-1),Scale3D=(X=1,Y=1,Z=1))
    bParentInitial=False
    bPropagateToChildren=True
    Weight=@n6.OutputParam
   ) {template="Set Relative Transform::Execute(in Child,in Parent,in bParentInitial,in Value,in Weight,in bPropagateToChildren)"}
   n7 = &RigUnit_GetTransform(Item=(Type=Bone,Name="BASE_R_C_collar_01"), bInitial=False)
   n8 = &RigUnit_GetTransform(Item=(Type=Control,Name="head_ctrl"), bInitial=False)
   > &RigUnit_SetRelativeTransformForItem(
    Child=(Type=Bone,Name="BASE_R_C_collar_01")
    Parent=(Type=Bone,Name="BASE_R_C_collar_01")
    Value=(Rotation=(X=0,Y=0.4226,Z=0,W=0.9063),Translation=(X=1,Y=0,Z=0),Scale3D=(X=1,Y=1,Z=1))
    bParentInitial=False
    bPropagateToChildren=True
    Weight=&RigVMFunction_MathDoubleRemap(
     SourceMaximum=8
     SourceMinimum=12
     TargetMaximum=1
     TargetMinimum=0
     Value=&RigVMFunction_MathVectorDistance(A=@n7.Transform.Translation, B=@n8.Transform.Translation).Result
    ).Result
   ) {template="Set Relative Transform::Execute(in Child,in Parent,in bParentInitial,in Value,in Weight,in bPropagateToChildren)"}
   n9 = &RigVMFunction_MathDoubleRemap(
    SourceMaximum=8
    SourceMinimum=12
    TargetMaximum=1
    TargetMinimum=0
    Value=&RigVMFunction_MathVectorDistance(
     A=&RigUnit_GetTransform(Item=(Type=Bone,Name="BASE_L_C_collar_01"), bInitial=False).Transform.Translation
     B=@n8.Transform.Translation
    ).Result
   )
   > &RigUnit_SetRelativeTransformForItem(
    Child=(Type=Bone,Name="BASE_L_C_collar_01")
    Parent=(Type=Bone,Name="BASE_L_C_collar_01")
    Value=(Rotation=(X=0,Y=0.4226,Z=0,W=0.9063),Translation=(X=-1,Y=0,Z=1),Scale3D=(X=1,Y=1,Z=1))
    bParentInitial=False
    bPropagateToChildren=True
    Weight=@n9.Result
   ) {template="Set Relative Transform::Execute(in Child,in Parent,in bParentInitial,in Value,in Weight,in bPropagateToChildren)"}
   > &RigUnit_ModifyTransforms(
    ItemToModify=((Item=(Type=Bone,Name="extra_L_C_collar_01"),Transform=(Rotation=(X=0,Y=0,Z=0,W=1),Translation=(X=0,Y=-2,Z=0),Scale3D=(X=1,Y=1,Z=1))))
    Weight=@n9.Result
   )
   n10 = &RigUnit_GetTransform(Item=(Type=Bone,Name="BASE_L_B_collar_01"), bInitial=False)
   > &RigUnit_SetRelativeTransformForItem(
    Child=(Type=Bone,Name="BASE_L_B_collar_01")
    Parent=(Type=Bone,Name="BASE_L_B_collar_01")
    Value=(Rotation=(X=0,Y=0.5,Z=0,W=0.866),Translation=(X=-1,Y=0,Z=-1),Scale3D=(X=1,Y=1,Z=1))
    bParentInitial=False
    bPropagateToChildren=True
    Weight=&RigVMFunction_MathDoubleRemap(
     SourceMaximum=4
     SourceMinimum=8
     TargetMaximum=1
     TargetMinimum=0
     Value=&RigVMFunction_MathVectorDistance(A=@n10.Transform.Translation, B=@n8.Transform.Translation).Result
    ).Result
   ) {template="Set Relative Transform::Execute(in Child,in Parent,in bParentInitial,in Value,in Weight,in bPropagateToChildren)"}
   n11 = &RigUnit_GetTransform(Item=(Type=Bone,Name="BASE_R_B_collar_01"), bInitial=False)
   > &RigUnit_SetRelativeTransformForItem(
    Child=(Type=Bone,Name="BASE_R_B_collar_01")
    Parent=(Type=Bone,Name="BASE_R_B_collar_01")
    Value=(Rotation=(X=0,Y=0.2588,Z=0,W=0.9659),Translation=(X=1,Y=0,Z=-1),Scale3D=(X=1,Y=1,Z=1))
    bParentInitial=False
    bPropagateToChildren=True
    Weight=&RigVMFunction_MathDoubleRemap(
     SourceMaximum=4
     SourceMinimum=8
     TargetMaximum=1
     TargetMinimum=0
     Value=&RigVMFunction_MathVectorDistance(A=@n11.Transform.Translation, B=@n8.Transform.Translation).Result
    ).Result
   ) {template="Set Relative Transform::Execute(in Child,in Parent,in bParentInitial,in Value,in Weight,in bPropagateToChildren)"}
   n12 = &RigUnit_GetTransform(Item=(Type=Bone,Name="BASE_L_A_collar_01"), bInitial=False)
   > &RigUnit_SetRelativeTransformForItem(
    Child=(Type=Bone,Name="BASE_L_A_collar_01")
    Parent=(Type=Bone,Name="BASE_L_A_collar_01")
    Value=(Rotation=(X=0,Y=0.2588,Z=0,W=0.9659),Translation=(X=1,Y=0,Z=-1),Scale3D=(X=1,Y=1,Z=1))
    bParentInitial=False
    bPropagateToChildren=True
    Weight=&RigVMFunction_MathDoubleRemap(
     SourceMaximum=4
     SourceMinimum=8
     TargetMaximum=1
     TargetMinimum=0
     Value=&RigVMFunction_MathVectorDistance(B=@n12.Transform.Translation, A=@n8.Transform.Translation).Result
    ).Result
   ) {template="Set Relative Transform::Execute(in Child,in Parent,in bParentInitial,in Value,in Weight,in bPropagateToChildren)"}
   n13 = &RigUnit_GetTransform(Item=(Type=Bone,Name="BASE_R_A_collar_01"), bInitial=False)
   > &RigUnit_SetRelativeTransformForItem(
    Child=(Type=Bone,Name="BASE_R_A_collar_01")
    Parent=(Type=Bone,Name="BASE_R_A_collar_01")
    Value=(Rotation=(X=0,Y=0.2588,Z=0,W=0.9659),Translation=(X=-1,Y=0,Z=1),Scale3D=(X=1,Y=1,Z=1))
    bParentInitial=False
    bPropagateToChildren=True
    Weight=&RigVMFunction_MathDoubleRemap(
     SourceMaximum=4
     SourceMinimum=8
     TargetMaximum=1
     TargetMinimum=0
     Value=&RigVMFunction_MathVectorDistance(B=@n13.Transform.Translation, A=@n8.Transform.Translation).Result
    ).Result
   ) {template="Set Relative Transform::Execute(in Child,in Parent,in bParentInitial,in Value,in Weight,in bPropagateToChildren)"}
   > n14 = &RigUnit_SphericalPoseReader(
    ActiveRegionScaleFactors=(PositiveWidth=1,NegativeWidth=0.1,PositiveHeight=0.5,NegativeHeight=1)
    ActiveRegionSize=0.3
    Debug=(bDrawDebug=False,bDraw2D=false,bDrawLocalAxes=false,DebugScale=30,DebugSegments=30,DebugThickness=0.2)
    DriverAxis=(X=-1,Y=0,Z=0)
    DriverItem=(Type=Bone,Name="upperarm_r")
    FalloffRegionScaleFactors=(PositiveWidth=1,NegativeWidth=0.5,PositiveHeight=1,NegativeHeight=1)
    FalloffSize=0.5
    OptionalParentItem=(Type=Bone,Name="pelvis")
    RotationOffset=(X=-120,Y=30,Z=0)
   )
   > &RigUnit_ModifyTransforms(
    ItemToModify=((Item=(Type=Bone,Name="BASE_R_A_collar_01"),Transform=(Rotation=(X=0,Y=-0.342,Z=0,W=0.9397),Translation=(X=0,Y=0,Z=0),Scale3D=(X=1,Y=1,Z=1))))
    Weight=@n14.OutputParam
   )
   n15 = &RigVMFunction_AlphaInterp(InterpSpeedDecreasing=5, InterpSpeedIncreasing=5, Scale=1, bClampResult=False, Value=RigVMVariableNode() {variable=KneeSliding}.Value)
   &RigUnit_ModifyTransforms(
    ItemToModify=((Item=(Type=Bone,Name="dyn_L_frontCoat_base_BASE"),Transform=(Rotation=(X=0,Y=-0.0872,Z=0,W=0.9962),Translation=(X=0,Y=0,Z=0),Scale3D=(X=1,Y=1,Z=1))))
    ExecutePin=@SequenceAggregate.D
    Weight=@n15.Result
   )
   > &RigUnit_ModifyTransforms(
    ItemToModify=((Item=(Type=Bone,Name="dyn_L_sideCoat_base"),Transform=(Rotation=(X=0,Y=-0.4226,Z=0,W=0.9063),Translation=(X=0,Y=0,Z=0),Scale3D=(X=1,Y=1,Z=1))))
    Weight=@n15.Result
   )
   > &RigUnit_ModifyTransforms(
    ItemToModify=((Item=(Type=Bone,Name="dyn_L_backCoat_base"),Transform=(Rotation=(X=0,Y=-0.7071,Z=0,W=0.7071),Translation=(X=0,Y=0,Z=0),Scale3D=(X=1,Y=1,Z=1))))
    Weight=@n15.Result
   )
   > &RigUnit_ModifyTransforms(
    ItemToModify=((Item=(Type=Bone,Name="dyn_R_backCoat_base"),Transform=(Rotation=(X=0,Y=-0.7071,Z=0,W=0.7071),Translation=(X=0,Y=0,Z=0),Scale3D=(X=1,Y=1,Z=1))))
    Weight=@n15.Result
   )
   > &RigUnit_ModifyTransforms(
    ItemToModify=((Item=(Type=Bone,Name="dyn_R_sideCoat_base"),Transform=(Rotation=(X=-0.25,Y=-0.433,Z=-0.433,W=0.75),Translation=(X=0,Y=4,Z=0),Scale3D=(X=1,Y=1,Z=1))))
    Weight=@n15.Result
   )
   > &RigUnit_ModifyTransforms(
    ItemToModify=((Item=(Type=Bone,Name="dyn_R_sideCoat_2"),Transform=(Rotation=(X=0,Y=-0.1736,Z=0,W=0.9848),Translation=(X=0,Y=0,Z=0),Scale3D=(X=1,Y=1,Z=1))))
    Weight=@n15.Result
   )
   > &RigUnit_ModifyTransforms(
    ItemToModify=((Item=(Type=Bone,Name="dyn_R_frontCoat_base_BASE"),Transform=(Rotation=(X=-0.0091,Y=-0.2711,Z=-0.3079,W=0.9119),Translation=(X=0,Y=0,Z=0),Scale3D=(X=1,Y=1,Z=1))))
    Weight=@n15.Result
   )
   > &RigUnit_ModifyTransforms(
    ItemToModify=((Item=(Type=Bone,Name="dyn_R_frontCoat_2"),Transform=(Rotation=(X=0,Y=-0.1736,Z=0,W=0.9848),Translation=(X=0,Y=0,Z=0),Scale3D=(X=1,Y=1,Z=1))))
    Weight=@n15.Result
   )
   n16 = &RigVMFunction_AlphaInterp(InterpSpeedDecreasing=5, InterpSpeedIncreasing=5, Scale=1, bClampResult=False, Value=RigVMVariableNode() {variable=BaseballSliding}.Value)
   > &RigUnit_ModifyTransforms(
    ItemToModify=((Item=(Type=Bone,Name="dyn_L_frontCoat_base_BASE"),Transform=(Rotation=(X=0.3686,Y=-0.2182,Z=-0.2742,W=0.861),Translation=(X=0,Y=0,Z=0),Scale3D=(X=1,Y=1,Z=1))))
    Weight=@n16.Result
   )
   > &RigUnit_ModifyTransforms(
    ItemToModify=((Item=(Type=Bone,Name="dyn_L_frontCoat_2"),Transform=(Rotation=(X=0,Y=0.1736,Z=0,W=0.9848),Translation=(X=0,Y=0,Z=0),Scale3D=(X=1,Y=1,Z=1))))
    Weight=1
   )
   > &RigUnit_ModifyTransforms(
    ItemToModify=((Item=(Type=Bone,Name="dyn_L_sideCoat_base"),Transform=(Rotation=(X=-0.0885,Y=-0.3304,Z=-0.2432,W=0.9077),Translation=(X=0,Y=0,Z=0),Scale3D=(X=1,Y=1,Z=1))))
    Weight=@n16.Result
   )
   > &RigUnit_ModifyTransforms(
    ItemToModify=((Item=(Type=Bone,Name="dyn_L_sideCoat_2"),Transform=(Rotation=(X=0,Y=0.0872,Z=0,W=0.9962),Translation=(X=0,Y=0,Z=0),Scale3D=(X=1,Y=1,Z=1))))
    Weight=1
   )
   > &RigUnit_ModifyTransforms(
    ItemToModify=((Item=(Type=Bone,Name="dyn_L_backCoat_base"),Transform=(Rotation=(X=0,Y=-0.7071,Z=0,W=0.7071),Translation=(X=0,Y=0,Z=0),Scale3D=(X=1,Y=1,Z=1))))
    Weight=@n16.Result
   )
   > &RigUnit_ModifyTransforms(
    ItemToModify=((Item=(Type=Bone,Name="dyn_R_backCoat_base"),Transform=(Rotation=(X=0,Y=-0.7065,Z=0,W=0.7077),Translation=(X=0,Y=0,Z=0),Scale3D=(X=1,Y=1,Z=1))))
    ItemToModify.0.Transform.Rotation=&RigVMFunction_MathQuaternionFromEuler(Euler=(X=10,Y=-110,Z=10), RotationOrder=XYZ).Result
    Weight=@n16.Result
   )
   > &RigUnit_ModifyTransforms(
    ItemToModify=((Item=(Type=Bone,Name="dyn_R_sideCoat_base"),Transform=(Rotation=(X=0,Y=0,Z=0,W=1),Translation=(X=0,Y=4,Z=0),Scale3D=(X=1,Y=1,Z=1))))
    ItemToModify.0.Transform.Rotation=&RigVMFunction_MathQuaternionFromEuler(Euler=(X=-40,Y=-120,Z=0), RotationOrder=XYZ).Result
    Weight=@n16.Result
   )
   > &RigUnit_ModifyTransforms(
    ItemToModify=((Item=(Type=Bone,Name="dyn_R_sideCoat_2"),Transform=(Rotation=(X=0,Y=0,Z=0,W=1),Translation=(X=0,Y=0,Z=0),Scale3D=(X=1,Y=1,Z=1))))
    Weight=@n16.Result
   )
   > &RigUnit_ModifyTransforms(
    ItemToModify=((Item=(Type=Bone,Name="dyn_R_frontCoat_base_BASE"),Transform=(Rotation=(X=0,Y=-0.7066,Z=0,W=0.7076),Translation=(X=0,Y=0,Z=0),Scale3D=(X=1,Y=1,Z=1))))
    ItemToModify.0.Transform.Rotation=&RigVMFunction_MathQuaternionFromEuler(Euler=(X=0,Y=-80,Z=0), RotationOrder=XZY).Result
    Weight=@n16.Result
   )
   > &RigUnit_ModifyTransforms(
    ItemToModify=((Item=(Type=Bone,Name="dyn_R_frontCoat_2"),Transform=(Rotation=(X=0,Y=-0.1736,Z=0,W=0.9848),Translation=(X=0,Y=0,Z=0),Scale3D=(X=1,Y=1,Z=1))))
    Weight=@n16.Result
   )
   &RigUnit_ModifyTransforms(
    ItemToModify=((Item=(Type=Bone,Name="knif"),Transform=(Rotation=(X=0.0868,Y=-0.0076,Z=-0.0868,W=0.9924),Translation=(X=0,Y=0,Z=0),Scale3D=(X=1,Y=1,Z=1))))
    Weight=1
    ExecutePin=@SequenceAggregate.E
   )
   n17 = &RigVMFunction_AlphaInterp(InterpSpeedDecreasing=5, InterpSpeedIncreasing=5, Scale=1, bClampResult=False, Value=RigVMVariableNode() {variable=crouching}.Value)
   > &RigUnit_ModifyTransforms(
    ItemToModify=((Item=(Type=Bone,Name="knif"),Transform=(Rotation=(X=-0.0259,Y=0.1349,Z=-0.1236,W=0.9828),Translation=(X=0,Y=-2,Z=0),Scale3D=(X=1,Y=1,Z=1))))
    Weight=@n17.Result
   )
   > &RigUnit_ModifyTransforms(
    ItemToModify="((Item=(Type=Bone,Name=\"dyn_L_backCoat_base\"),Transform=(Rotation=(X=0,Y=-0.6428,Z=0,W=0.766),Translation=(X=0,Y=0,Z=0),Scale3D=(X=1,Y=1,Z=1))),(Item=(Type=Bone,Name=\"dyn_R_backCoat_base\"),Transform=(Rotation=(X=0,Y=-0.6428,Z=0,W=0.766),Translation=(X=0,Y= ...(truncated)"
    Weight=0
   )
   n18 = RigVMVariableNode() {variable=tactical_sprinting}
   n19 = &RigVMFunction_AlphaInterp(InterpSpeedDecreasing=5, InterpSpeedIncreasing=5, Scale=1, bClampResult=False, Value=@n18.Value)
   > &RigUnit_ModifyTransforms(
    ItemToModify="((Item=(Type=Bone,Name=\"dyn_L_backCoat_base\"),Transform=(Rotation=(X=0,Y=-0.1736,Z=0,W=0.9848),Translation=(X=0,Y=0,Z=0),Scale3D=(X=1,Y=1,Z=1))),(Item=(Type=Bone,Name=\"dyn_R_backCoat_base\"),Transform=(Rotation=(X=0,Y=-0.1736,Z=0,W=0.9848),Translation=(X=0, ...(truncated)"
    Weight=@n19.Result
   )
   > &RigUnit_ModifyTransforms(
    ItemToModify=((Item=(Type=Bone,Name="waist_reduction"),Transform=(Rotation=(X=0,Y=0,Z=0,W=1),Translation=(X=0,Y=0,Z=0),Scale3D=(X=1,Y=0.1,Z=0.1))))
    Weight=1
   )
   n20 = &RigVMFunction_KalmanFloat(BufferSize=9, Value=RigVMVariableNode() {variable=dbno}.Value) {template="KalmanFilter::Execute(in Value,in BufferSize,out Result)"}
   > &RigUnit_ModifyTransforms(
    ItemToModify="((Item=(Type=Bone,Name=\"dyn_L_frontCoat_base_BASE\"),Transform=(Rotation=(X=0,Y=0,Z=0.5,W=0.866),Translation=(X=0,Y=0,Z=0),Scale3D=(X=1,Y=1,Z=1))),(Item=(Type=Bone,Name=\"None\"),Transform=(Rotation=(X=0,Y=0,Z=0.2588,W=0.9659),Translation=(X=0,Y=0,Z=0),Scale3 ...(truncated)"
    Weight=@n20.Result
   )
   > &RigUnit_ModifyTransforms(
    ItemToModify="((Item=(Type=Bone,Name=\"BASE_L_rope\"),Transform=(Rotation=(X=0,Y=0,Z=0.5,W=0.866),Translation=(X=0,Y=0,Z=0),Scale3D=(X=1,Y=1,Z=1))),(Item=(Type=Bone,Name=\"BASE_R_rope\"),Transform=(Rotation=(X=0,Y=0,Z=0,W=1),Translation=(X=0,Y=0,Z=0),Scale3D=(X=1,Y=1,Z=1))) ...(truncated)"
    Weight=@n20.Result
   )
   n21 = RigVMVariableNode() {variable=RBAN}
   &RigUnit_AimConstraintLocalSpaceOffset(
    AdvancedSettings=(DebugSettings=(bEnabled=true,Scale=10,WorldOffset=(Rotation=(X=0,Y=0,Z=0,W=1),Translation=(X=0,Y=0,Z=0),Scale3D=(X=1,Y=1,Z=1))),RotationOrderForFilter=XZY)
    Child=(Type=Bone,Name="dyn_R_frontCoat_base")
    Filter=(bX=True,bY=True,bZ=True)
    Parents=((Item=(Type=Bone,Name="calf_r"),Weight=1))
    UpAxis=(X=0,Y=1,Z=0)
    WorldUp=(Target=(X=0,Y=0,Z=1),Kind=Location,Space=(Type=Bone,Name="calf_r"))
    ExecutePin=@SequenceAggregate.F
    Weight=@n21.Value
   )
   > &RigUnit_AimConstraintLocalSpaceOffset(
    AdvancedSettings=(DebugSettings=(bEnabled=true,Scale=10,WorldOffset=(Rotation=(X=0,Y=0,Z=0,W=1),Translation=(X=0,Y=0,Z=0),Scale3D=(X=1,Y=1,Z=1))),RotationOrderForFilter=XZY)
    Child=(Type=Bone,Name="dyn_L_frontCoat_base_BASE")
    Filter=(bX=True,bY=True,bZ=True)
    Parents=((Item=(Type=Bone,Name="calf_l"),Weight=1))
    UpAxis=(X=0,Y=1,Z=0)
    WorldUp=(Target=(X=0,Y=0,Z=1),Kind=Location,Space=(Type=Bone,Name="calf_l"))
    Weight=@n21.Value
   )
   &RigVMFunction_AlphaInterp(InterpSpeedDecreasing=5, InterpSpeedIncreasing=5, Scale=1, bClampResult=False, Value=RigVMVariableNode() {variable=jumping}.Value)
   &RigVMFunction_KalmanFloat(BufferSize=9, Value=@n1.OutputParam) {template="KalmanFilter::Execute(in Value,in BufferSize,out Result)"}
   &RigUnit_ModifyTransforms(
    ExecutePin=()
    ItemToModify=((Item=(Type=Bone,Name="knif"),Transform=(Rotation=(X=-0.238,Y=0.1585,Z=-0.0915,W=0.9539),Translation=(X=0,Y=-2,Z=0),Scale3D=(X=1,Y=1,Z=1))))
    Weight=1
   )
   &RigVMFunction_MathDoubleMul(B=&RigVMFunction_MathDoubleMul(B=0, A=RigVMVariableNode() {variable=jumping}.Value).Result, A=@n18.Value)
   Add = &RigVMFunction_MathDoubleAdd(B=0.1, A=@n2.OutputParam)
   Multiply_4 = RigVMAggregateNode(B=@Add.Result, A=@n3.OutputParam, C=@n4.Value)
   n22 = RigVMVariableNode() {variable=crouching}
   RigVMRerouteNode(Value=&RigVMFunction_KalmanFloat(BufferSize=9, Value=&RigVMFunction_MathDoubleMul(A=@n22.Value, B=@n1.OutputParam).Result) {template="KalmanFilter::Execute(in Value,in BufferSize,out Result)"}.Result)
  ]}
]
hierarchy [
 root {loc=<0 0 0>, rot=<0 0 0 1>}
 pelvis {parent=root, loc=<0 -0.7071 93.1191>, rot=<0.0127 0.707 -0.0127 -0.707>}
 spine_01 {parent=pelvis, loc=<6.7494 -0.6388 0>, rot=<0 0 0.034 -0.9994>}
 spine_02 {parent=spine_01, loc=<7.252 -0.035 0>, rot=<0 0 0.0238 -0.9997>}
 spine_03 {parent=spine_02, loc=<6.0106 0 0>, rot=<0 0 0.0149 -0.9999>}
 spine_04 {parent=spine_03, loc=<8.1023 0 0>, rot=<0 0 -0.0132 -0.9999>}
 spine_05 {parent=spine_04, loc=<10.0142 0 0>, rot=<0 0 -0.0015 -1>}
 attach_backpack {parent=spine_05, loc=<-0.123 -13.473 0>, rot=<-0.0093 -0.707 0.0093 -0.707>}
 neck_01 {parent=spine_05, loc=<11.3138 0.1529 0>, rot=<0 0 -0.2003 -0.9797>}
 neck_02 {parent=neck_01, loc=<4.5627 0 0>, rot=<0 0 0.0504 -0.9987>}
 head {parent=neck_02, loc=<5.3914 0 0>, rot=<0 0 0.115 -0.9934>}
 head_ctrl {controlType=EULER_TRANSFORM, parent=head, type=Control, loc=<0 0 0>, rot=<0 0 0 1>}
 FACIAL_C_FacialRoot {parent=head, loc=<2.1524 -0.746 -0.0002>, rot=<0.4879 -0.5118 -0.4879 -0.5118>}
 FACIAL_C_Forehead {parent=FACIAL_C_FacialRoot, loc=<0.0021 -3.4056 9.8816>, rot=<-0.0334 0 0 -0.9994>}
 FACIAL_L_ForeheadIn {parent=FACIAL_C_FacialRoot, loc=<1.4295 -3.3424 9.7854>, rot=<-0.0507 -0.0696 0.005 -0.9963>}
 FACIAL_R_ForeheadIn {parent=FACIAL_C_FacialRoot, loc=<-1.4275 -3.3429 9.7845>, rot=<-0.0507 0.0695 -0.0049 -0.9963>}
 FACIAL_L_ForeheadMid {parent=FACIAL_C_FacialRoot, loc=<2.7623 -3.5596 9.3186>, rot=<-0.0612 -0.1759 -0.0188 -0.9823>}
 FACIAL_R_ForeheadMid {parent=FACIAL_C_FacialRoot, loc=<-2.7605 -3.5592 9.3188>, rot=<-0.0612 0.1758 0.0189 -0.9823>}
 FACIAL_L_ForeheadOut {parent=FACIAL_C_FacialRoot, loc=<3.9721 -3.5326 8.5929>, rot=<-0.1118 -0.2564 -0.0766 -0.957>}
 FACIAL_R_ForeheadOut {parent=FACIAL_C_FacialRoot, loc=<-3.97 -3.5324 8.5926>, rot=<-0.1118 0.2564 0.0766 -0.957>}
 FACIAL_L_EyesackUpper {parent=FACIAL_C_FacialRoot, loc=<2.8127 -2.5259 9.8182>, rot=<0.0588 -0.0849 -0.0637 -0.9926>}
 FACIAL_R_EyesackUpper {parent=FACIAL_C_FacialRoot, loc=<-2.8079 -2.5264 9.8177>, rot=<0.0588 0.0849 0.0637 -0.9926>}
 FACIAL_L_EyelidUpperFurrow {parent=FACIAL_C_FacialRoot, loc=<2.704 -2.3503 9.6691>, rot=<-0.0202 -0.1516 -0.0781 -0.9851>}
 FACIAL_R_EyelidUpperFurrow {parent=FACIAL_C_FacialRoot, loc=<-2.6957 -2.3571 9.6687>, rot=<-0.0202 0.1516 0.0782 -0.9851>}
 FACIAL_L_EyelidUpperA {parent=FACIAL_C_FacialRoot, loc=<2.779 -1.9467 7.9695>, rot=<-0.1379 0 0 -0.9904>}
 FACIAL_R_EyelidUpperA {parent=FACIAL_C_FacialRoot, loc=<-2.7781 -1.9473 7.9705>, rot=<-0.1379 0 0 -0.9904>}
 FACIAL_L_EyelidLowerA {parent=FACIAL_C_FacialRoot, loc=<2.779 -1.9467 7.9695>, rot=<0.2936 0 0 -0.9559>}
 FACIAL_R_EyelidLowerA {parent=FACIAL_C_FacialRoot, loc=<-2.7781 -1.9473 7.9705>, rot=<0.2936 0 0 -0.9559>}
 FACIAL_L_EyeCornerInner {parent=FACIAL_C_FacialRoot, loc=<1.8597 -1.5312 8.8124>, rot=<0.01 0 0 -0.9999>}
 FACIAL_R_EyeCornerInner {parent=FACIAL_C_FacialRoot, loc=<-1.8596 -1.5419 8.8112>, rot=<0.01 0 0 -1>}
 FACIAL_L_EyeCornerOuter {parent=FACIAL_C_FacialRoot, loc=<4.0835 -1.8649 8.4257>, rot=<0.0092 -0.3826 0.0039 -0.9239>}
 FACIAL_R_EyeCornerOuter {parent=FACIAL_C_FacialRoot, loc=<-4.0811 -1.8618 8.4307>, rot=<0.0092 0.3826 -0.0038 -0.9239>}
 FACIAL_L_EyesackLower {parent=FACIAL_C_FacialRoot, loc=<3.1124 -0.6896 8.3296>, rot=<-0.0124 -0.1399 -0.0208 -0.9899>}
 FACIAL_R_EyesackLower {parent=FACIAL_C_FacialRoot, loc=<-3.1056 -0.6955 8.3285>, rot=<-0.0124 0.1399 0.0209 -0.9899>}
 FACIAL_L_CheekInner {parent=FACIAL_C_FacialRoot, loc=<1.9385 0.2559 8.536>, rot=<-0.1135 -0.09 -0.1962 -0.9698>}
 FACIAL_R_CheekInner {parent=FACIAL_C_FacialRoot, loc=<-1.9405 0.2578 8.5328>, rot=<-0.1135 0.0899 0.1962 -0.9698>}
 FACIAL_L_CheekOuter {parent=FACIAL_C_FacialRoot, loc=<3.4791 0.5067 7.066>, rot=<0.0644 -0.3759 0.1041 -0.9185>}
 FACIAL_R_CheekOuter {parent=FACIAL_C_FacialRoot, loc=<-3.4796 0.5056 7.0673>, rot=<0.0644 0.3759 -0.1041 -0.9185>}
 FACIAL_L_NasolabialBulge {parent=FACIAL_C_FacialRoot, loc=<3.2451 3.8861 7.0258>, rot=<0.0672 -0.3771 0.0276 -0.9233>}
 FACIAL_R_NasolabialBulge {parent=FACIAL_C_FacialRoot, loc=<-3.2578 3.8862 7.04>, rot=<0.0672 0.3771 -0.0276 -0.9233>}
 FACIAL_L_NasolabialFurrow {parent=FACIAL_C_FacialRoot, loc=<2.6528 2.5567 8.96>, rot=<-0.024 -0.1295 -0.3623 -0.9227>}
 FACIAL_R_NasolabialFurrow {parent=FACIAL_C_FacialRoot, loc=<-2.6693 2.561 8.9742>, rot=<-0.024 0.1294 0.3624 -0.9227>}
 FACIAL_L_CheekLower {parent=FACIAL_C_FacialRoot, loc=<3.5024 3.8233 6.3526>, rot=<0.0436 -0.4263 0.062 -0.9014>}
 FACIAL_R_CheekLower {parent=FACIAL_C_FacialRoot, loc=<-3.5148 3.8212 6.3672>, rot=<0.0436 0.4263 -0.062 -0.9014>}
 FACIAL_C_Nose {parent=FACIAL_C_FacialRoot, loc=<0.0199 0.0583 10.3821>, rot=<-0.1204 0 0 -0.9927>}
 FACIAL_C_NoseLower {parent=FACIAL_C_Nose, loc=<-0.0226 2.7608 -0.4442>, rot=<0.1305 0 0 -0.9914>}
 FACIAL_L_Nostril {parent=FACIAL_C_Nose, loc=<0.9754 2.0494 -0.4284>, rot=<0.0173 0.1305 0.0023 -0.9913>}
 FACIAL_R_Nostril {parent=FACIAL_C_Nose, loc=<-1.0241 2.0707 -0.4219>, rot=<0.0173 -0.1305 -0.0023 -0.9913>}
 FACIAL_C_NoseTip {parent=FACIAL_C_Nose, loc=<-0.0135 2.0741 0.5127>, rot=<0.0872 0 0 -0.9962>}
 FACIAL_C_MouthUpper {parent=FACIAL_C_FacialRoot, loc=<-0.0076 4.1379 7.8661>, rot=<0.01 0 0 -0.9999>}
 FACIAL_C_LipUpper {parent=FACIAL_C_MouthUpper, loc=<0.0004 -0.3122 2.1681>, rot=<-0.0872 0 0 -0.9962>}
 FACIAL_L_LipUpper {parent=FACIAL_C_MouthUpper, loc=<1.1065 -0.2314 1.898>, rot=<-0.0778 -0.1601 -0.0379 -0.9833>}
 FACIAL_R_LipUpper {parent=FACIAL_C_MouthUpper, loc=<-1.1227 -0.2281 1.9229>, rot=<-0.0778 0.1601 0.0379 -0.9833>}
 FACIAL_L_LipUpperOuter {parent=FACIAL_C_MouthUpper, loc=<1.9302 -0.0231 1.2766>, rot=<-0.0984 -0.2293 -0.0983 -0.9634>}
 FACIAL_R_LipUpperOuter {parent=FACIAL_C_MouthUpper, loc=<-1.9476 -0.0002 1.305>, rot=<-0.0984 0.2293 0.0983 -0.9634>}
 FACIAL_L_LipCorner {parent=FACIAL_C_MouthUpper, loc=<2.5378 0.2905 0.6003>, rot=<0 -0.2078 0 -0.9782>}
 FACIAL_R_LipCorner {parent=FACIAL_C_MouthUpper, loc=<-2.4978 0.3297 0.552>, rot=<0 0.2078 0 -0.9782>}
 FACIAL_L_Masseter {parent=FACIAL_C_FacialRoot, loc=<5.7083 1.2243 4.998>, rot=<0.0793 0.1129 0.0068 -0.9904>}
 FACIAL_R_Masseter {parent=FACIAL_C_FacialRoot, loc=<-5.7074 1.229 4.9918>, rot=<0.0793 -0.113 -0.0067 -0.9904>}
 FACIAL_C_UnderChin {parent=FACIAL_C_FacialRoot, loc=<0.0005 8.4585 5.2777>, rot=<0.5085 0 0 -0.8611>}
 FACIAL_L_UnderChin {parent=FACIAL_C_FacialRoot, loc=<2.822 7.1576 3.813>, rot=<0.4109 -0.354 0.0859 -0.8357>}
 FACIAL_R_UnderChin {parent=FACIAL_C_FacialRoot, loc=<-2.8229 7.1591 3.809>, rot=<0.4109 0.354 -0.0859 -0.8357>}
 FACIAL_C_LowerLipRotation {parent=FACIAL_C_FacialRoot, loc=<-0.0018 1.8561 3.3059>, rot=<0.3514 0 0 -0.9362>}
 FACIAL_C_MouthLower {parent=FACIAL_C_LowerLipRotation, loc=<-0.006 -1.2826 4.9353>, rot=<-0.342 0 0 -0.9397>}
 FACIAL_C_LipLower {parent=FACIAL_C_MouthLower, loc=<0.0089 1.1642 1.797>, rot=<0.0872 0 0 -0.9962>}
 FACIAL_L_LipLower {parent=FACIAL_C_MouthLower, loc=<0.9362 1.0938 1.6053>, rot=<0.0927 -0.1294 0.0272 -0.9869>}
 FACIAL_R_LipLower {parent=FACIAL_C_MouthLower, loc=<-0.9298 1.1127 1.6132>, rot=<0.0927 0.1294 -0.0272 -0.9869>}
 FACIAL_L_LipLowerOuter {parent=FACIAL_C_MouthLower, loc=<1.7518 0.7331 1.0405>, rot=<0.0913 -0.2321 0.0318 -0.9679>}
 FACIAL_R_LipLowerOuter {parent=FACIAL_C_MouthLower, loc=<-1.7588 0.7732 1.0541>, rot=<0.0913 0.2321 -0.0318 -0.9679>}
 FACIAL_C_Jaw {parent=FACIAL_C_FacialRoot, loc=<-0.0018 1.8561 3.3059>, rot=<0.3514 0 0 -0.9362>}
 FACIAL_C_Jawline {parent=FACIAL_C_Jaw, loc=<0.0013 2.437 7.3705>, rot=<0.2586 0 0 -0.966>}
 FACIAL_L_ChinSide {parent=FACIAL_C_Jaw, loc=<2.4649 0.269 6.6585>, rot=<-0.2208 -0.3738 0.0499 -0.8995>}
 FACIAL_R_ChinSide {parent=FACIAL_C_Jaw, loc=<-2.4862 0.2809 6.6657>, rot=<-0.2208 0.3738 -0.0499 -0.8995>}
 FACIAL_C_Chin {parent=FACIAL_C_Jaw, loc=<-0.0004 1.0624 7.3479>, rot=<-0.2591 0 0 -0.9659>}
 FACIAL_L_Ear {parent=FACIAL_C_FacialRoot, loc=<6.5549 -0.6382 1.6146>, rot=<-0.0163 -0.001 -0.0261 -0.9995>}
 FACIAL_R_Ear {parent=FACIAL_C_FacialRoot, loc=<-6.5554 -0.6377 1.5982>, rot=<-0.0163 0.0009 0.0262 -0.9995>}
 FACIAL_L_Jawline {parent=FACIAL_C_FacialRoot, loc=<3.8442 5.2538 5.2563>, rot=<0.219 0.1572 -0.0562 -0.9614>}
 FACIAL_R_Jawline {parent=FACIAL_C_FacialRoot, loc=<-3.8429 5.2493 5.2513>, rot=<0.219 -0.1572 0.0563 -0.9613>}
 FACIAL_C_NoseBridge {parent=FACIAL_C_FacialRoot, loc=<-0.0005 -1.9806 10.5039>, rot=<0.0538 0 0 -0.9986>}
 FACIAL_L_EyelidUpperB {parent=FACIAL_C_FacialRoot, loc=<2.779 -1.9467 7.9695>, rot=<-0.2152 0 0 -0.9766>}
 FACIAL_L_EyelidLowerB {parent=FACIAL_C_FacialRoot, loc=<2.779 -1.9467 7.9695>, rot=<0.3268 0 0 -0.9451>}
 FACIAL_R_EyelidLowerB {parent=FACIAL_C_FacialRoot, loc=<-2.7781 -1.9473 7.9705>, rot=<0.335 0 0 -0.9422>}
 FACIAL_R_EyelidUpperB {parent=FACIAL_C_FacialRoot, loc=<-2.7781 -1.9473 7.9705>, rot=<-0.2152 0 0 -0.9766>}
 clavicle_l {parent=spine_05, loc=<4.3575 -0.1466 -2.3223>, rot=<0.0465 -0.7289 -0.009 -0.6829>}
 upperarm_l {parent=clavicle_l, loc=<15.5111 0 0>, rot=<0.0226 -0.3234 -0.0088 -0.9459>}
 lowerarm_l {parent=upperarm_l, loc=<26.6514 0 0>, rot=<0.0391 -0.0099 -0.2453 -0.9686>}
 hand_l {parent=lowerarm_l, loc=<22.3491 0 0>, rot=<-0.7081 -0.0507 0.0975 -0.6975>}
 thumb_01_l {parent=hand_l, loc=<1.3118 -0.3191 -2.459>, rot=<0.6546 -0.4042 0.0664 -0.6354>}
 thumb_02_l {parent=thumb_01_l, loc=<5.9336 -0.0448 0.0401>, rot=<-0.0037 0.0213 0.298 -0.9543>}
 thumb_03_l {parent=thumb_02_l, loc=<2.2223 0.3682 -0.0907>, rot=<0.0118 -0.013 0.0066 -0.9998>}
 index_metacarpal_l {parent=hand_l, loc=<1.6535 0.0183 -2.96>, rot=<0.0158 -0.0124 -0.0173 -0.9996>}
 index_01_l {parent=index_metacarpal_l, loc=<8.3008 0.0357 -0.0524>, rot=<-0.0044 0.0485 0.0924 -0.9945>}
 index_02_l {parent=index_01_l, loc=<3.3587 -0.3004 0.0452>, rot=<-0.002 0.0027 0.3225 -0.9466>}
 index_03_l {parent=index_02_l, loc=<2.5578 -0.123 -0.0252>, rot=<-0.002 0.0067 0.2558 -0.9667>}
 middle_metacarpal_l {parent=hand_l, loc=<2.1856 -0.1725 -0.5984>, rot=<0.006 0.0168 -0.0329 -0.9993>}
 middle_01_l {parent=middle_metacarpal_l, loc=<7.3275 -0.2447 0.0129>, rot=<-0.0208 0.0042 0.0858 -0.9961>}
 middle_02_l {parent=middle_01_l, loc=<4.0123 0.1401 0.1175>, rot=<-0.0218 0.0258 0.3512 -0.9357>}
 middle_03_l {parent=middle_02_l, loc=<2.6077 0.0534 0.1373>, rot=<0.0082 0.0188 0.1708 -0.9851>}
 ring_metacarpal_l {parent=hand_l, loc=<2.2074 -0.3625 1.3638>, rot=<-0.0284 0.0722 -0.0458 -0.9959>}
 ring_01_l {parent=ring_metacarpal_l, loc=<7.0295 -0.6197 -0.1446>, rot=<-0.0489 -0.0252 0.1892 -0.9804>}
 ring_02_l {parent=ring_01_l, loc=<3.6174 0.4647 -0.1027>, rot=<0 0 0.2806 -0.9598>}
 ring_03_l {parent=ring_02_l, loc=<2.4152 0.114 0.0199>, rot=<0 0 0.1189 -0.9929>}
 pinky_metacarpal_l {parent=hand_l, loc=<1.7616 -0.9753 3.3814>, rot=<-0.0397 0.0814 -0.069 -0.9935>}
 pinky_01_l {parent=pinky_metacarpal_l, loc=<7.1709 -0.4831 -0.1094>, rot=<-0.0654 -0.0035 0.2242 -0.9723>}
 pinky_02_l {parent=pinky_01_l, loc=<2.8274 0.0893 -0.1043>, rot=<0.0037 0.0011 0.2854 -0.9584>}
 pinky_03_l {parent=pinky_02_l, loc=<2.2046 0.0307 -0.0078>, rot=<-0.0002 0.0098 0.1445 -0.9895>}
 deform_wrist_dn_base_l {parent=hand_l, loc=<0 -2.2342 0>, rot=<0 0 0 -1>}
 deform_wrist_dn_l {parent=deform_wrist_dn_base_l, loc=<0 0 0>, rot=<0 0 0 1>}
 deform_wrist_up_base_l {parent=hand_l, loc=<0 2.2682 0>, rot=<0 0 0 -1>}
 deform_wrist_up_l {parent=deform_wrist_up_base_l, loc=<0 0 0>, rot=<0 0 0 1>}
 lowerarm_twist_02_l {parent=lowerarm_l, loc=<9.6199 0 0>, rot=<0 0 0 -1>}
 lowerarm_twist_01_l {parent=lowerarm_l, loc=<17.8549 0 0>, rot=<0 0 0 1>}
 deform_elbow_in_base_l {parent=lowerarm_l, loc=<0.2112 5.259 0.8274>, rot=<0 0 0 1>}
 deform_elbow_in_l {parent=deform_elbow_in_base_l, loc=<0 0 0>, rot=<0 0 0 1>}
 deform_elbow_out_base_l {parent=lowerarm_l, loc=<0 -4 0>, rot=<0 0 0 1>}
 deform_elbow_out_l {parent=deform_elbow_out_base_l, loc=<0 0 0>, rot=<0 0 0 1>}
 upperarm_twist_01_l {parent=upperarm_l, loc=<0.0041 0 0>, rot=<0 0 0 -1>}
 deform_upperarm_bk_base_l {parent=upperarm_twist_01_l, loc=<1.0813 -7.2193 -1.9453>, rot=<0 0 0 -1>}
 deform_upperarm_bk_l {parent=deform_upperarm_bk_base_l, loc=<0 0 0>, rot=<0 0 0 1>}
 deform_upperarm_fr_base_l {parent=upperarm_twist_01_l, loc=<-0.3691 7.0506 -2.765>, rot=<0 0 0 -1>}
 deform_upperarm_fr_l {parent=deform_upperarm_fr_base_l, loc=<0 0 0>, rot=<0 0 0 1>}
 deform_upperarm_up_base_l {parent=upperarm_twist_01_l, loc=<-3.2216 -1.1391 3.3311>, rot=<0 0 0 1>}
 deform_upperarm_up_l {parent=deform_upperarm_up_base_l, loc=<0 0 0>, rot=<0 0 0 1>}
 upperarm_twist_02_l {parent=upperarm_l, loc=<13.7941 0 0>, rot=<0 0 0 -1>}
 attach_cape {parent=spine_05, loc=<9.017 -11.714 0>, rot=<0.0616 -0.7044 -0.0616 -0.7044>}
 clavicle_r {parent=spine_05, loc=<4.3575 -0.1466 2.3223>, rot=<0.7289 0.0465 -0.6829 0.009>}
 upperarm_r {parent=clavicle_r, loc=<-15.5111 0 0>, rot=<0.0226 -0.3234 -0.0088 -0.9459>}
 lowerarm_r {parent=upperarm_r, loc=<-26.6514 0 0>, rot=<0.0391 -0.0099 -0.2453 -0.9686>}
 hand_r {parent=lowerarm_r, loc=<-22.3491 0 0>, rot=<-0.7081 -0.0507 0.0975 -0.6975>}
 thumb_01_r {parent=hand_r, loc=<-1.3118 0.3191 2.459>, rot=<0.6546 -0.4042 0.0664 -0.6354>}
 thumb_02_r {parent=thumb_01_r, loc=<-5.9336 0.0448 -0.0401>, rot=<-0.0037 0.0213 0.298 -0.9543>}
 thumb_03_r {parent=thumb_02_r, loc=<-2.2223 -0.3682 0.0907>, rot=<0.0118 -0.013 0.0066 -0.9998>}
 index_metacarpal_r {parent=hand_r, loc=<-1.6535 -0.0183 2.96>, rot=<0.0158 -0.0124 -0.0173 -0.9996>}
 index_01_r {parent=index_metacarpal_r, loc=<-8.3008 -0.0357 0.0524>, rot=<-0.0044 0.0485 0.0924 -0.9945>}
 index_02_r {parent=index_01_r, loc=<-3.3587 0.3004 -0.0452>, rot=<-0.002 0.0027 0.3225 -0.9466>}
 index_03_r {parent=index_02_r, loc=<-2.5578 0.123 0.0252>, rot=<-0.002 0.0067 0.2558 -0.9667>}
 middle_metacarpal_r {parent=hand_r, loc=<-2.1856 0.1725 0.5984>, rot=<0.006 0.0168 -0.0329 -0.9993>}
 middle_01_r {parent=middle_metacarpal_r, loc=<-7.3275 0.2447 -0.0129>, rot=<-0.0208 0.0042 0.0858 -0.9961>}
 middle_02_r {parent=middle_01_r, loc=<-4.0123 -0.1401 -0.1175>, rot=<-0.0218 0.0258 0.3512 -0.9357>}
 middle_03_r {parent=middle_02_r, loc=<-2.6077 -0.0534 -0.1373>, rot=<0.0082 0.0188 0.1708 -0.9851>}
 ring_metacarpal_r {parent=hand_r, loc=<-2.2074 0.3625 -1.3638>, rot=<-0.0284 0.0722 -0.0458 -0.9959>}
 ring_01_r {parent=ring_metacarpal_r, loc=<-7.0295 0.6197 0.1446>, rot=<-0.0489 -0.0252 0.1892 -0.9804>}
 ring_02_r {parent=ring_01_r, loc=<-3.6174 -0.4647 0.1027>, rot=<0 0 0.2806 -0.9598>}
 ring_03_r {parent=ring_02_r, loc=<-2.4152 -0.114 -0.0199>, rot=<0 0 0.1189 -0.9929>}
 pinky_metacarpal_r {parent=hand_r, loc=<-1.7616 0.9753 -3.3814>, rot=<-0.0397 0.0814 -0.069 -0.9935>}
 pinky_01_r {parent=pinky_metacarpal_r, loc=<-7.1709 0.4831 0.1094>, rot=<-0.0654 -0.0035 0.2242 -0.9723>}
 pinky_02_r {parent=pinky_01_r, loc=<-2.8274 -0.0893 0.1043>, rot=<0.0037 0.0011 0.2854 -0.9584>}
 pinky_03_r {parent=pinky_02_r, loc=<-2.2046 -0.0307 0.0078>, rot=<-0.0002 0.0098 0.1445 -0.9895>}
 deform_wrist_dn_base_r {parent=hand_r, loc=<-0.0004 2.2345 -0.0001>, rot=<0 0 0 -1>}
 deform_wrist_dn_r {parent=deform_wrist_dn_base_r, loc=<0 0 0>, rot=<0 0 0 1>}
 deform_wrist_up_base_r {parent=hand_r, loc=<0.0003 -2.2684 0.0001>, rot=<0 0 0 -1>}
 deform_wrist_up_r {parent=deform_wrist_up_base_r, loc=<0 0 0>, rot=<0 0 0 1>}
 lowerarm_twist_02_r {parent=lowerarm_r, loc=<-9.6199 0 0>, rot=<0 0 0 -1>}
 lowerarm_twist_01_r {parent=lowerarm_r, loc=<-17.8549 0 0>, rot=<0 0 0 -1>}
 deform_elbow_in_base_r {parent=lowerarm_r, loc=<-0.2111 -5.2591 -0.8275>, rot=<0 0 0 -1>}
 deform_elbow_in_r {parent=deform_elbow_in_base_r, loc=<-0.0001 0 0.0002>, rot=<0 0 0 1>}
 deform_elbow_out_base_r {parent=lowerarm_r, loc=<0 4 -0.0001>, rot=<0 0 0 -1>}
 deform_elbow_out_r {parent=deform_elbow_out_base_r, loc=<0 0 0>, rot=<0 0 0 1>}
 upperarm_twist_01_r {parent=upperarm_r, loc=<-0.0041 0 0>, rot=<0 0 0 -1>}
 deform_upperarm_bk_base_r {parent=upperarm_twist_01_r, loc=<-1.0814 7.2192 1.9455>, rot=<0 0 0 -1>}
 deform_upperarm_bk_r {parent=deform_upperarm_bk_base_r, loc=<0 0 0>, rot=<0 0 0 1>}
 deform_upperarm_fr_base_r {parent=upperarm_twist_01_r, loc=<0.3689 -7.0506 2.7652>, rot=<0 0 0 -1>}
 deform_upperarm_fr_r {parent=deform_upperarm_fr_base_r, loc=<0 0 0>, rot=<0 0 0 1>}
 deform_upperarm_up_base_r {parent=upperarm_twist_01_r, loc=<3.2214 1.1391 -3.3308>, rot=<0 0 0 1>}
 deform_upperarm_up_r {parent=deform_upperarm_up_base_r, loc=<0 0 0>, rot=<0 0 0 -1>}
 upperarm_twist_02_r {parent=upperarm_r, loc=<-13.7941 0 0>, rot=<0 0 0 -1>}
 pin_to_geo_zipper {parent=spine_05, loc=<-0.6755 12.8634 0>, rot=<-0.158 -0.9874 0 0>}
 zipper {parent=pin_to_geo_zipper, loc=<0.05 0 0>, rot=<0 0 0 1>}
 BASE_L_A_collar_01 {parent=spine_05, loc=<8.9934 7.3889 -4.9505>, rot=<0.0905 -0.3367 -0.1895 -0.9179>}
 L_A_collar_01 {parent=BASE_L_A_collar_01, loc=<0 0 0>, rot=<0 0 0 1>}
 L_A_collar_02 {parent=L_A_collar_01, loc=<5.9081 0 0>, rot=<0 -0.2042 0 -0.9789>}
 BASE_R_C_collar_01 {parent=spine_05, loc=<15.5744 -1.7386 4.6247>, rot=<-0.1734 -0.2886 0.9364 -0.0993>}
 R_C_collar_01 {parent=BASE_R_C_collar_01, loc=<0 0 0>, rot=<0 0 0 1>}
 R_C_collar_02 {parent=R_C_collar_01, loc=<-3.5712 -0.0001 0.0002>, rot=<0 -0.1973 0 -0.9804>}
 extra_R_C_collar_01 {parent=BASE_R_C_collar_01, loc=<0 0 0>, rot=<0 0 0 1>}
 BASE_R_A_collar_01 {parent=spine_05, loc=<8.9938 7.3889 4.9505>, rot=<-0.3367 -0.0905 0.9179 -0.1895>}
 R_A_collar_01 {parent=BASE_R_A_collar_01, loc=<0 0 0>, rot=<0 0 0 1>}
 R_A_collar_02 {parent=R_A_collar_01, loc=<-5.9079 -0.0001 0.0002>, rot=<0 -0.2042 0 -0.9789>}
 BASE_R_B_collar_01 {parent=spine_05, loc=<13.0503 2.6166 6.1542>, rot=<-0.3262 -0.0884 0.9217 -0.1905>}
 R_B_collar_01 {parent=BASE_R_B_collar_01, loc=<0 0 0>, rot=<0 0 0 1>}
 R_B_collar_02 {parent=R_B_collar_01, loc=<-4.3127 -0.0002 0.0003>, rot=<0 -0.0908 0 -0.9959>}
 BASE_M_collar_01 {parent=spine_05, loc=<16.5811 -3.268 0>, rot=<0.6996 -0.103 0.103 -0.6996>}
 M_collar_01 {parent=BASE_M_collar_01, loc=<0 0 0>, rot=<0 0 0 1>}
 M_collar_02 {parent=M_collar_01, loc=<3.2188 0 0>, rot=<0 -0.0911 0 -0.9958>}
 extra_M_collar_01 {parent=BASE_M_collar_01, loc=<0 0 0>, rot=<0 0 0 1>}
 BASE_L_C_collar_01 {parent=spine_05, loc=<15.574 -1.7386 -4.6247>, rot=<0.2886 -0.1734 -0.0993 -0.9364>}
 L_C_collar_01 {parent=BASE_L_C_collar_01, loc=<0 0 0>, rot=<0 0 0 1>}
 L_C_collar_02 {parent=L_C_collar_01, loc=<3.5716 0 0>, rot=<0 -0.1973 0 -0.9804>}
 extra_L_C_collar_01 {parent=BASE_L_C_collar_01, loc=<0 0 0>, rot=<0 0 0 1>}
 BASE_L_B_collar_01 {parent=spine_05, loc=<13.0501 2.6166 -6.1542>, rot=<0.0884 -0.3262 -0.1905 -0.9217>}
 L_B_collar_01 {parent=BASE_L_B_collar_01, loc=<0 0 0>, rot=<0 0 0 1>}
 L_B_collar_02 {parent=L_B_collar_01, loc=<4.3131 0 0>, rot=<0 -0.0908 0 -0.9959>}
 deform_pec_aim_base_l {parent=spine_04, loc=<7.8728 12.6388 -2.4123>, rot=<0.0102 -0.4602 0.3979 -0.7936>}
 deform_pec_aim_l {parent=deform_pec_aim_base_l, loc=<0 0 0>, rot=<0 0 0 -1>}
 deform_pec_base_l {parent=deform_pec_aim_l, loc=<10.766 0 0>, rot=<0 0 0 1>}
 deform_pec_l {parent=deform_pec_base_l, loc=<0.055 0 0>, rot=<0 0 0 1>}
 deform_pec_aim_base_r {parent=spine_04, loc=<7.8723 12.6388 2.4123>, rot=<-0.4602 -0.0102 0.7936 0.3979>}
 deform_pec_aim_r {parent=deform_pec_aim_base_r, loc=<0 0 0>, rot=<0 0 0 -1>}
 deform_pec_base_r {parent=deform_pec_aim_r, loc=<-10.7662 -0.0006 -0.0007>, rot=<0 0 0 1>}
 deform_pec_r {parent=deform_pec_base_r, loc=<0 0 0>, rot=<0 0 0 1>}
 thigh_l {parent=pelvis, loc=<-2.5693 -0.0109 -7.7501>, rot=<0.0012 0.0572 0.0155 -0.9982>}
 calf_l {parent=thigh_l, loc=<-39.0484 0 0>, rot=<0.0006 -0.0161 -0.0391 -0.9991>}
 foot_l {parent=calf_l, loc=<-42.561 0.1077 0.4292>, rot=<0.0418 -0.0455 0.0343 -0.9975>}
 ball_l {parent=foot_l, loc=<-6.1599 13.8264 0.6806>, rot=<0.0001 -0.0001 -0.7063 -0.7079>}
 calf_twist_02_l {parent=calf_l, loc=<-18.8064 -0.0967 0.0397>, rot=<0.0237 0.0095 -0.0035 -0.9997>}
 calf_twist_01_l {parent=calf_l, loc=<-35.813 -0.0967 0.0397>, rot=<0.0237 0.0095 -0.0035 -0.9997>}
 deform_knee_in_base_l {parent=calf_l, loc=<1.298 -5.2957 -0.0496>, rot=<0 0 0 -1>}
 deform_knee_in_l {parent=deform_knee_in_base_l, loc=<0 0 0>, rot=<0 0 0 1>}
 deform_knee_out_base_l {parent=calf_l, loc=<0 4 0>, rot=<0 0 0 -1>}
 deform_knee_out_l {parent=deform_knee_out_base_l, loc=<0 0 0>, rot=<0 0 0 1>}
 thigh_twist_01_l {parent=thigh_l, loc=<-21.6303 0 0>, rot=<0.0013 0 0 -1>}
 deform_groin_base_l {parent=thigh_l, loc=<-2.1908 8.1142 -1.7442>, rot=<0 0 0 -1>}
 deform_groin_l {parent=deform_groin_base_l, loc=<0 0 0>, rot=<0 0 0 1>}
 thigh_r {parent=pelvis, loc=<-2.5693 -0.0109 7.7501>, rot=<0.0572 -0.0012 0.9982 0.0155>}
 calf_r {parent=thigh_r, loc=<39.0484 0 0>, rot=<0.0006 -0.0161 -0.0391 -0.9991>}
 foot_r {parent=calf_r, loc=<42.561 -0.1077 -0.4292>, rot=<0.0418 -0.0455 0.0343 -0.9975>}
 ball_r {parent=foot_r, loc=<6.1599 -13.8264 -0.6806>, rot=<0.0001 -0.0001 -0.7063 -0.7079>}
 calf_twist_02_r {parent=calf_r, loc=<18.8064 0.0967 -0.0397>, rot=<0.0237 0.0095 -0.0035 -0.9997>}
 calf_twist_01_r {parent=calf_r, loc=<35.813 0.0967 -0.0397>, rot=<0.0237 0.0095 -0.0035 -0.9997>}
 deform_knee_in_base_r {parent=calf_r, loc=<-1.298 5.2957 0.0496>, rot=<0 0 0 -1>}
 deform_knee_in_r {parent=deform_knee_in_base_r, loc=<0 0 0>, rot=<0 0 0 1>}
 deform_knee_out_base_r {parent=calf_r, loc=<0 -4 0>, rot=<0 0 0 -1>}
 deform_knee_out_r {parent=deform_knee_out_base_r, loc=<0 0 0>, rot=<0 0 0 1>}
 thigh_twist_01_r {parent=thigh_r, loc=<21.6303 0 0>, rot=<0.0013 0 0 -1>}
 deform_groin_base_r {parent=thigh_r, loc=<2.1908 -8.1142 1.7442>, rot=<0 0 0 -1>}
 deform_groin_r {parent=deform_groin_base_r, loc=<0 0 0>, rot=<0 0 0 1>}
 deform_glute_base_l {parent=pelvis, loc=<-8.8064 -8.8222 -8.4287>, rot=<0.0275 0.0501 -0.4472 -0.8926>}
 deform_glute_l {parent=deform_glute_base_l, loc=<0 0 0>, rot=<0 0 0 1>}
 deform_glute_base_r {parent=pelvis, loc=<-8.8062 -8.8222 8.4287>, rot=<0.0501 -0.0275 0.8926 -0.4472>}
 deform_glute_r {parent=deform_glute_base_r, loc=<0 0 0>, rot=<0 0 0 1>}
 dyn_R_sideCoat_base {parent=pelvis, loc=<3.2411 -3.0049 14.6858>, rot=<0.9457 0.042 -0.1462 -0.2873>}
 dyn_R_sideCoat_1 {parent=dyn_R_sideCoat_base, loc=<-0.1 0.0056 -0.0035>, rot=<-0.0003 0.0219 -0.0163 -0.9996>}
 dyn_R_sideCoat_2 {parent=dyn_R_sideCoat_1, loc=<-18.1566 -0.4422 0.8366>, rot=<0.006 -0.0414 0.0081 -0.9991>}
 dyn_R_sideCoat_3 {parent=dyn_R_sideCoat_2, loc=<-16.7926 -0.2662 -0.374>, rot=<0 0 0 -1>}
 dyn_L_sideCoat_base {parent=pelvis, loc=<3.2205 -2.958 -14.6936>, rot=<0.0118 -0.9409 0.3044 -0.148>}
 dyn_L_sideCoat_1 {parent=dyn_L_sideCoat_base, loc=<0.1 0 0>, rot=<-0.0006 0.0186 -0.0036 -0.9998>}
 dyn_L_sideCoat_2 {parent=dyn_L_sideCoat_1, loc=<18.3824 0.559 -0.4568>, rot=<0.0009 -0.0171 -0.0057 -0.9998>}
 dyn_R_backCoat_base {parent=pelvis, loc=<6.7901 -10.5094 6.0781>, rot=<0.698 0.0532 -0.0598 -0.7116>}
 dyn_R_backCoat_1 {parent=dyn_R_backCoat_base, loc=<-0.1 0 0>, rot=<0 0 0 1>}
 dyn_R_backCoat_2 {parent=dyn_R_backCoat_1, loc=<-13.9404 -0.0171 -0.2133>, rot=<0.0007 -0.0071 -0.0129 -0.9999>}
 dyn_R_backCoat_3 {parent=dyn_R_backCoat_2, loc=<-18.289 0.0609 -0.0071>, rot=<0 0 0 1>}
 dyn_L_backCoat_base {parent=pelvis, loc=<7.0208 -10.5423 -6.078>, rot=<0.0558 -0.7063 0.7033 -0.0572>}
 dyn_L_backCoat_1 {parent=dyn_L_backCoat_base, loc=<0.1 0 0>, rot=<0 0 0 1>}
 dyn_L_backCoat_2 {parent=dyn_L_backCoat_1, loc=<14.1628 0.0028 0.1442>, rot=<0.0007 -0.0071 -0.0129 -0.9999>}
 dyn_L_backCoat_3 {parent=dyn_L_backCoat_2, loc=<18.2893 -0.0507 0.0071>, rot=<0 0 0 1>}
 knif {parent=pelvis, loc=<10.2577 5.1583 -11.7996>, rot=<0.3064 -0.9071 -0.2341 -0.1691>}
 knife_handle {parent=knif, loc=<-4.5 0 0>, rot=<0 0 0 1>}
 knife_blade {parent=knif, loc=<9 0 0>, rot=<0 0 0 1>}
 BASE_R_rope {parent=pelvis, loc=<4.9558 11.8309 2.4599>, rot=<-0.0393 -0.9499 -0.0174 0.3096>}
 R_rope_01 {parent=BASE_R_rope, loc=<0 0 0>, rot=<0 0 0 1>}
 R_rope_02 {parent=R_rope_01, loc=<4.8 0 0>, rot=<0 0 0 1>}
 BASE_L_rope {parent=pelvis, loc=<4.8679 11.5065 -3.0375>, rot=<-0.1068 -0.9677 0.0785 -0.2145>}
 L_rope_01 {parent=BASE_L_rope, loc=<0 0 0>, rot=<0 0 0 1>}
 L_rope_02 {parent=L_rope_01, loc=<5 0 0>, rot=<0 0 0 1>}
 dyn_R_frontCoat_base_BASE {parent=pelvis, loc=<9.8684 4.2979 11.0974>, rot=<0.9885 0.0267 -0.1206 0.0873>}
 dyn_R_frontCoat_base {parent=dyn_R_frontCoat_base_BASE, loc=<0 0 0>, rot=<0 0 0 1>}
 dyn_R_frontCoat_1 {parent=dyn_R_frontCoat_base, loc=<-10.2075 -0.0005 0.0002>, rot=<0 0 0 1>}
 dyn_R_frontCoat_2 {parent=dyn_R_frontCoat_1, loc=<-15.5909 0 0>, rot=<0.0007 -0.0071 -0.0129 -0.9999>}
 dyn_R_frontCoat_3 {parent=dyn_R_frontCoat_2, loc=<-16.4888 0 0>, rot=<0 0 0 1>}
 dyn_L_frontCoat_base_BASE {parent=pelvis, loc=<9.8938 4.302 -11.1097>, rot=<0.0322 -0.9823 -0.0873 -0.1629>}
 dyn_L_frontCoat_base {parent=dyn_L_frontCoat_base_BASE, loc=<0 0 0>, rot=<0 0 0 1>}
 dyn_L_frontCoat_1 {parent=dyn_L_frontCoat_base, loc=<10.2072 0 0>, rot=<0 0 0 1>}
 dyn_L_frontCoat_2 {parent=dyn_L_frontCoat_1, loc=<15.5909 0 0>, rot=<0.0007 -0.0071 -0.0129 -0.9999>}
 dyn_L_frontCoat_3 {parent=dyn_L_frontCoat_2, loc=<16.4887 0 0>, rot=<0 0 0 1>}
 waist_reduction {parent=pelvis, loc=<0 0 0>, rot=<0 0 0 1>}
 M_back_belt {parent=pelvis, loc=<11.2407 -10.8487 0.0103>, rot=<0.491 -0.5089 0.5089 -0.491>}
 R_3_belt {parent=pelvis, loc=<10.7095 -7.2739 10.1671>, rot=<0.6495 -0.2797 0.6591 -0.2562>}
 R_2_belt {parent=pelvis, loc=<10.2411 -0.4089 12.5531>, rot=<0.707 -0.0127 0.707 0.0127>}
 R_1_belt {parent=pelvis, loc=<8.5834 7.0929 9.4826>, rot=<0.635 0.311 0.6235 0.3336>}
 M_front_belt {parent=pelvis, loc=<7.8702 9.021 0>, rot=<-0.5089 -0.491 -0.491 -0.5089>}
 L_3_belt {parent=pelvis, loc=<10.7092 -7.2739 -10.1671>, rot=<-0.6495 0.2797 0.6591 -0.2562>}
 L_1_belt {parent=pelvis, loc=<8.5835 7.0929 -9.4825>, rot=<0.635 0.311 -0.6235 -0.3336>}
 L_2_belt {parent=pelvis, loc=<10.2408 -0.4088 -12.5531>, rot=<0.707 -0.0127 -0.707 -0.0127>}
]
variables [
 BaseballSliding {type=double, default=0}
 KneeSliding {type=double, default=0}
 RBAN {type=double, default=0}
 crouching {type=double, default=0}
 dbno {type=double, default=0}
 jumping {type=double, default=0}
 moving_2d {type=double, default=0}
 not_moving {type=double, default=0}
 static_crouching_rotation {type=double, default=0}
 tactical_sprinting {type=double, default=0}
]
