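# Post-process style cosmetic AnimBP for the CubicVice Veil male body, adding secondary motion to coat, collar and zipper.
# Copies the parent mesh pose through a pre-sim Control Rig, a weighted zipper-pin constraint, coat and collar RigidBody sims, zipper AnimDynamics, then a post-sim Control Rig.
# Update event turns movement states into rig floats and computes coat gravity override.
# Coat simulation switches off while sliding; an extra coat RigidBody node sits unconnected.
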AnimBlueprint /BRCosmetics/Characters/Player/Male/Medium/Bodies/M_MED_CubicVice_Veil/Meshes/M_MED_CubicVice_Veil_AnimBP
target_skeleton=&/M_MED_CubicVice_Veil_Skeleton
graphs [
 AnimGraph {nodes [
   n1 = NOT Boolean(A=Get bIsSliding())
   n2 = Get Gravity Override()
   RigidBody!(ComponentPose, bAlphaBoolEnabled:bool=@n1, OverrideWorldGravity:Vector=@n2, ExternalForce) -> (Pose) {
    alphaInputType=Bool
    bOverrideWorldGravity=true
    overridePhysicsAsset=&/M_MED_CubicVice_Veil_Physics_Coat
    simulationSpace=BaseBoneSpace
    alphaBoolBlend {blendOption=Linear}
    baseBoneRef {boneName=pelvis}
    componentAppliedLinearAccClamp=<1000 1000 1000>
    componentLinearAccScale=<0.5 0.5 0.5>
    componentLinearVelScale=<0.5 0.5 0.5>
    overrideWorldGravity {z=-980}
    simSpaceSettings {
     maxAngularAcceleration=200
     maxAngularVelocity=1
     maxLinearAcceleration=100
     maxLinearVelocity=50
     velocityScaleZ=0.2
     worldAlpha=1
     externalAngularVelocity=<0.2 0.2 0.2>
    }
   }
   n3 = Control Rig!(
    Source=Copy Pose From Mesh() {bUseAttachedParent=true}
    Alpha
    KneeSliding:double=Get KneeSliding_float() -> (KneeSliding_float:double)
    BaseballSliding:double=Get BaseballSliding_float() -> (BaseballSliding_float:double)
    crouching:double=Get crouching_float() -> (crouching_float:double)
    jumping:double=Get jumping_float() -> (jumping_float:double)
    tactical_sprinting:double=Get tactical_sprinting_float() -> (tactical_sprinting_float:double)
    not_moving:double=Get not_moving() -> (not_moving:double)
    RBAN:double=Get RBAN_float() -> (RBAN_float:double)
    static_crouching_rotation:double=Get static_crouching_rotation() -> (static_crouching_rotation:double)
    dbno:double=Get dbno() -> (dbno:double)
   ) -> (Pose) {
    alphaBoolBlend {blendOption=Linear}
    controlRigAssetReference {blueprintRigClass=&/M_MED_CubicVice_Veil_presim_CtrlRig_v002.M_MED_CubicVice_Veil_presim_CtrlRig_v002_C}
   }
   n4 = "Constraint - pin_to_geo_zipper (PPPPPPP)"!(
    ComponentPose=Local To Component(LocalPose=@n3)
    Alpha=1
    ConstraintWeights_0=0.0074
    ConstraintWeights_1:float=0.4775
    ConstraintWeights_2:float=0.2332
    ConstraintWeights_3:float=0.0095
    ConstraintWeights_4:float=0.1924
    ConstraintWeights_5:float=0.0789
    ConstraintWeights_6:float=0.0011
   ) -> (Pose) {
    alphaBoolBlend {blendOption=Linear}
    boneToModify {boneName=pin_to_geo_zipper}
    constraintSetup [
     {targetBone {boneName=clavicle_l}}
     {targetBone {boneName=spine_04}}
     {targetBone {boneName=spine_05}}
     {targetBone {boneName=deform_pec_l}}
     {targetBone {boneName=spine_03}}
     {targetBone {boneName=spine_02}}
     {targetBone {boneName=spine_01}}
    ]
    constraintWeights [1,1,1,1,1,1,1]
   }
   n5 = RigidBody(ComponentPose=@n4, bAlphaBoolEnabled=@n1, OverrideWorldGravity=@n2) {
    alphaInputType=Bool
    bClampLinearTranslationLimitToRefPose=true
    bOverrideWorldGravity=true
    overridePhysicsAsset=&M_MED_CubicVice_Veil_Physics_Coat
    simulationSpace=BaseBoneSpace
    alphaBoolBlend {blendInTime=0.2, blendOption=Linear, blendOutTime=0.2}
    baseBoneRef {boneName=pelvis}
    componentAppliedLinearAccClamp=<6000 6000 6000>
    componentLinearAccScale=<0.25 0.25 0.25>
    componentLinearVelScale=<0.25 0.25 0.25>
    overrideWorldGravity {z=-980}
    simSpaceSettings {
     maxAngularAcceleration=500
     maxAngularVelocity=3
     maxLinearAcceleration=400
     maxLinearVelocity=250
     velocityScaleZ=0
     worldAlpha=1
     externalLinearDragV=<0.1 0.1>
    }
   }
   n6 = RigidBody!(ComponentPose=@n5, Alpha:float, OverrideWorldGravity="0,0,0", ExternalForce) -> (Pose) {
    bOverrideWorldGravity=true
    overridePhysicsAsset=&/M_MED_CubicVice_Veil_Physics_Collar
    simulationSpace=BaseBoneSpace
    alphaBoolBlend {blendOption=Linear}
    baseBoneRef {boneName=spine_05}
    componentAppliedLinearAccClamp=<1000 1000 1000>
    componentLinearAccScale=<0.5 0.5 0.5>
    componentLinearVelScale=<0.5 0.5 0.5>
    overrideWorldGravity {z=-980}
    simSpaceSettings {
     maxAngularAcceleration=200
     maxAngularVelocity=1
     maxLinearAcceleration=100
     maxLinearVelocity=50
     velocityScaleZ=0.2
     worldAlpha=1
     externalAngularVelocity=<0.2 0.2 0.2>
    }
   }
   n7 = Anim Dynamics - Bone: zipper(ComponentPose=@n6, Alpha=1, ExternalForce="0,0,0") {
    angularDampingOverride=1
    angularSpringConstant=20000
    bAngularSpring=true
    bOverrideAngularDamping=true
    bUseGravityOverride=true
    gravityScale=0.35
    simulationSpace=RootRelative
    alphaBoolBlend {blendOption=Linear}
    boundBone {boneName=zipper}
    componentLinearAccScale=<1 1 1>
    componentLinearVelScale=<1 1 1>
    gravityOverride {z=-980}
    physicsBodyDefinitions [{
      boundBone {boneName=zipper}
      boxExtents=<20 30 30>
      constraintSetup {angularLimitsMax=<5 20 50>, angularLimitsMin=<-5 -20 -2>, angularTarget=<1 0 0>}
      localJointOffset=<10 0 0>
     }]
   }
   n8 = Control Rig!(
    Source=Component To Local(ComponentPose=@n7)
    Alpha
    not_moving=Get not_moving()
    crouching=Get crouching_float()
    static_crouching_rotation=Get static_crouching_rotation()
    dbno=Get dbno()
    rban:double=Get RBAN_float()
   ) -> (Pose) {
    alphaBoolBlend {blendOption=Linear}
    controlRigAssetReference {blueprintRigClass=&/M_MED_CubicVice_Veil_CtrlRig_postsim.M_MED_CubicVice_Veil_CtrlRig_postsim_C}
   }
   Output Pose(Result=@n8) {layerGroup=None}
  ]}
 EventGraph {nodes [
   Event BlueprintUpdateAnimation()
   > Set KneeSliding_float(KneeSliding_float:double="To Float (Boolean)"(InBool=Get bIsKneeSliding())) -> (Output_Get:double)
   > Set BaseballSliding_float(BaseballSliding_float:double="To Float (Boolean)"(InBool=Get bIsBaseballSliding())) -> (Output_Get:double)
   > Set crouching_float(crouching_float:double="To Float (Boolean)"(InBool=Get bIsCrouching())) -> (Output_Get:double)
   n1 = NOT Boolean(A=Get bIsSkyDiving())
   n2 = Get bIsJumping()
   > Set jumping_float(jumping_float:double="To Float (Boolean)"(InBool=AND Boolean(A=@n2, B=@n1))) -> (Output_Get:double)
   > Set tactical_sprinting_float(tactical_sprinting_float:double="To Float (Boolean)"(InBool=Get bIsTacticalSprinting())) -> (Output_Get:double)
   n3 = Get bIsMoving2D()
   > Set not_moving(not_moving:double="To Float (Boolean)"(InBool=NOT Boolean(A=@n3))) -> (Output_Get:double)
   > Set dbno(dbno:double="To Float (Boolean)"(InBool=Get bIsDBNO())) -> (Output_Get:double)
   n4 = Get bRigidBodyEnabled()
   > Set RBAN_float(RBAN_float:double="To Float (Boolean)"(InBool=NOT Boolean(A=@n4))) -> (Output_Get:double)
   n5 = BreakRotator(InRot=Get PawnDeltaRotation() -> (PawnDeltaRotation:Rotator))
   n6 = AND Boolean!(A="float > float"(A="Absolute (Float)"(A=@n5.Yaw), B=3), B=Get bIsCrouching(), C:bool=NOT Boolean(A=Get bIsMoving2D())) -> (ReturnValue)
   > Set static_crouching_rotation(static_crouching_rotation:double="To Float (Boolean)"(InBool=NOT Boolean(A=@n6))) -> (Output_Get:double)
   > n7 = Gravity Override State Function DEPRECATED(CharacterPartAnimInstance=Self-Reference(), GameTimeInSeconds=GetGameTimeInSeconds(), GravityOverrideParams=Get Gravity Override Params())
   > Set Gravity Override(Gravity Override=@n7)
   AND Boolean!(
    A=false
    B=NOT Boolean(A=Get bIsPlayingEmote())
    C=NOT Boolean(A=Get bIsCrouching())
    D:bool=NOT Boolean(A=Get bIsJumping())
    E:bool=NOT Boolean(A=Get bIsSkyDiving())
   ) -> (ReturnValue)
  ]}
]
variables [
 BaseballSliding_float {type=number}
 crouching_float {type=number}
 dbno {type=number}
 Gravity Override {type=Vector}
 Gravity Override Params {type=GravityOverrideParamsStruct, default {skydive_GoingUp {gravityOverride=<0 -300 600>}}}
 jumping_float {type=number}
 KneeSliding_float {type=number}
 NewVar_0 {type=boolean}
 not_moving {type=number}
 RBAN_float {type=number}
 static_crouching_rotation {type=number}
 tactical_sprinting_float {type=number}
]
